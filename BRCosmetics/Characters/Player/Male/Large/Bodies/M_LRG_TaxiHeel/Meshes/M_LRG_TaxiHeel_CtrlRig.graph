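# Control Rig for the large male TaxiHeel player body, fitting its retargeted limbs and driving corrective deform bones.
# Arm soft two-bone IK follows hand IK bones, switching pole-vector handling when gauntlets are worn.
# Swing-twist extraction spreads rotation along arm twist chains.
# Shoulder, elbow, knee, glute and groin correctives blend parent constraints, then offset bones using spherical pose readers or remapped joint yaw.
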ControlRigBlueprint /BRCosmetics/Characters/Player/Male/Large/Bodies/M_LRG_TaxiHeel/Meshes/M_LRG_TaxiHeel_CtrlRig
graphs [
 CollapseNode_ContainedGraph {nodes [
   n1 = RigVMFunctionEntryNode()
   RigVMFunction_Sequence = &RigVMFunction_Sequence(ExecuteContext=@n1.ExecuteContext)
   RigVMFunction_Sequence_5 = RigVMAggregateNode(ExecuteContext=@RigVMFunction_Sequence.A)
   &RigUnit_ParentConstraint(
    AdvancedSettings=(RotationOrderForFilter=XZY)
    Child=(Type=Bone,Name="deform_upperarm_up_base_l")
    Parents=((Item=(Type=Bone,Name="upperRetArm_twist_01_l"),Weight=0.5),(Item=(Type=Bone,Name="clavicle_l"),Weight=0.5))
    Weight=1
    ExecutePin=@RigVMFunction_Sequence_5.A
   )
   > n2 = &RigUnit_SphericalPoseReader(
    ActiveRegionScaleFactors=(PositiveWidth=2.5,NegativeWidth=3,PositiveHeight=3,NegativeHeight=3)
    ActiveRegionSize=0.1
    Debug=(bDrawDebug=false,DebugScale=25,DebugSegments=20,DebugThickness=0.25)
    DriverAxis=(X=1,Y=0,Z=0)
    DriverItem=(Type=Bone,Name="upperRetArm_l")
    FalloffRegionScaleFactors=(PositiveWidth=1,NegativeWidth=1,PositiveHeight=1,NegativeHeight=1)
    FalloffSize=0.5
    OptionalParentItem=(Type=Bone)
    RotationOffset=(X=0,Y=0,Z=0)
   )
   > &RigUnit_ModifyTransforms(
    ItemToModify=((Item=(Type=Bone,Name="deform_upperarm_up_l"),Transform=(Rotation=(X=0,Y=0,Z=0,W=1),Translation=(X=0,Y=0,Z=8),Scale3D=(X=0.8,Y=1,Z=1))))
    Weight=@n2.OutputParam
   )
   &RigUnit_ParentConstraint(
    AdvancedSettings=(RotationOrderForFilter=XZY)
    Child=(Type=Bone,Name="deform_upperarm_fr_base_l")
    Parents=((Item=(Type=Bone,Name="upperRetArm_twist_01_l"),Weight=0.3),(Item=(Type=Bone,Name="clavicle_l"),Weight=0.7))
    Weight=1
    ExecutePin=@RigVMFunction_Sequence_5.B
   )
   > n3 = &RigUnit_SphericalPoseReader(
    ActiveRegionScaleFactors=(PositiveWidth=0.5,NegativeWidth=3,PositiveHeight=1.5,NegativeHeight=1.5)
    ActiveRegionSize=0.1
    Debug=(bDrawDebug=false,DebugScale=25,DebugSegments=20,DebugThickness=0.25)
    DriverAxis=(X=1,Y=0,Z=0)
    DriverItem=(Type=Bone,Name="upperarm_l")
    FalloffRegionScaleFactors=(PositiveWidth=1,NegativeWidth=1,PositiveHeight=1,NegativeHeight=1)
    FalloffSize=0.4
    OptionalParentItem=(Type=Bone)
    RotationOffset=(X=90,Y=0,Z=0)
   )
   > &RigUnit_ModifyTransforms(
    ItemToModify="((Item=(Type=Bone,Name=\"deform_upperarm_fr_l\"),Transform=(Rotation=(X=0,Y=0,Z=0,W=1),Translation=(X=0,Y=4,Z=0),Scale3D=(X=0.8,Y=1,Z=1))),(Item=(Type=Control,Name=\"arm_l_pv_ctrl\"),Transform=(Rotation=(X=0,Y=0,Z=0,W=1),Translation=(X=12.8845,Y=0,Z=60.0285),S ...(truncated)"
    Weight=@n3.OutputParam
   )
   > &RigUnit_ModifyTransforms(
    ItemToModify="((Item=(Type=Bone,Name=\"deform_upperarm_bk_l\"),Transform=(Rotation=(X=0,Y=0,Z=0,W=1),Translation=(X=0,Y=-3,Z=0),Scale3D=(X=1,Y=1,Z=1))),(Item=(Type=Bone,Name=\"deform_upperarm_fr_l\"),Transform=(Rotation=(X=0,Y=0,Z=0,W=1),Translation=(X=4.7909,Y=0,Z=14),Scal ...(truncated)"
    Weight=@n3.OutputParam
   )
   &RigUnit_ParentConstraint(
    AdvancedSettings=(RotationOrderForFilter=XZY)
    Child=(Type=Bone,Name="deform_upperarm_bk_base_l")
    Parents=((Item=(Type=Bone,Name="upperRetArm_twist_01_l"),Weight=0.3),(Item=(Type=Bone,Name="clavicle_l"),Weight=0.7))
    Weight=1
    ExecutePin=@RigVMFunction_Sequence_5.C
   )
   > n4 = &RigUnit_SphericalPoseReader(
    ActiveRegionScaleFactors=(PositiveWidth=3,NegativeWidth=1,PositiveHeight=1.5,NegativeHeight=1.5)
    ActiveRegionSize=0.1
    Debug=(bDrawDebug=false,DebugScale=25,DebugSegments=20,DebugThickness=0.25)
    DriverAxis=(X=1,Y=0,Z=0)
    DriverItem=(Type=Bone,Name="upperarm_l")
    FalloffRegionScaleFactors=(PositiveWidth=1,NegativeWidth=1,PositiveHeight=1,NegativeHeight=1)
    FalloffSize=0.4
    OptionalParentItem=(Type=Bone)
    RotationOffset=(X=90,Y=0,Z=180)
   )
   > &RigUnit_ModifyTransforms(
    ItemToModify=((Item=(Type=Bone,Name="deform_upperarm_bk_l"),Transform=(Rotation=(X=0,Y=0,Z=0,W=1),Translation=(X=0,Y=3,Z=0),Scale3D=(X=1,Y=1,Z=1))))
    Weight=@n4.OutputParam
   )
   RigVMFunction_Sequence_1 = RigVMAggregateNode(ExecuteContext=@RigVMFunction_Sequence.B)
   &RigUnit_ParentConstraint(
    AdvancedSettings=(RotationOrderForFilter=XZY)
    Child=(Type=Bone,Name="deform_upperarm_up_base_r")
    Parents=((Item=(Type=Bone,Name="upperRetArm_twist_01_r"),Weight=0.5),(Item=(Type=Bone,Name="clavicle_r"),Weight=0.5))
    Weight=1
    ExecutePin=@RigVMFunction_Sequence_1.A
   )
   > n5 = &RigUnit_SphericalPoseReader(
    ActiveRegionScaleFactors=(PositiveWidth=2.5,NegativeWidth=3,PositiveHeight=3,NegativeHeight=3)
    ActiveRegionSize=0.1
    Debug=(bDrawDebug=false,DebugScale=25,DebugSegments=20,DebugThickness=0.25)
    DriverAxis=(X=-1,Y=0,Z=0)
    DriverItem=(Type=Bone,Name="upperRetArm_r")
    FalloffRegionScaleFactors=(PositiveWidth=1,NegativeWidth=1,PositiveHeight=1,NegativeHeight=1)
    FalloffSize=0.5
    FlipHeightScaling=true
    OptionalParentItem=(Type=Bone)
    RotationOffset=(X=180,Y=0,Z=180)
   )
   > &RigUnit_ModifyTransforms(
    ItemToModify=((Item=(Type=Bone,Name="deform_upperarm_up_r"),Transform=(Rotation=(X=0,Y=0,Z=0,W=1),Translation=(X=0,Y=0,Z=-8),Scale3D=(X=0.8,Y=1,Z=1))))
    Weight=@n5.OutputParam
   )
   &RigUnit_ParentConstraint(
    AdvancedSettings=(RotationOrderForFilter=XZY)
    Child=(Type=Bone,Name="deform_upperarm_fr_base_r")
    Parents=((Item=(Type=Bone,Name="upperarm_twist_01_r"),Weight=0.3),(Item=(Type=Bone,Name="clavicle_r"),Weight=0.7))
    Weight=1
    ExecutePin=@RigVMFunction_Sequence_1.B
   )
   > n6 = &RigUnit_SphericalPoseReader(
    ActiveRegionScaleFactors=(PositiveWidth=0.5,NegativeWidth=3,PositiveHeight=1.5,NegativeHeight=1.5)
    ActiveRegionSize=0.1
    Debug=(bDrawDebug=false,DebugScale=25,DebugSegments=20,DebugThickness=0.25)
    DriverAxis=(X=-1,Y=0,Z=0)
    DriverItem=(Type=Bone,Name="upperarm_r")
    FalloffRegionScaleFactors=(PositiveWidth=1,NegativeWidth=1,PositiveHeight=1,NegativeHeight=1)
    FalloffSize=0.4
    FlipWidthScaling=true
    OptionalParentItem=(Type=Bone)
    RotationOffset=(X=90,Y=180,Z=180)
   )
   > &RigUnit_ModifyTransforms(
    ItemToModify=((Item=(Type=Bone,Name="deform_upperarm_fr_r"),Transform=(Rotation=(X=0,Y=0,Z=0,W=1),Translation=(X=0,Y=0,Z=0),Scale3D=(X=0.8,Y=1,Z=1))))
    Weight=@n6.OutputParam
   )
   > &RigUnit_ModifyTransforms(
    ItemToModify="((Item=(Type=Bone,Name=\"deform_upperarm_bk_r\"),Transform=(Rotation=(X=0,Y=0,Z=0,W=1),Translation=(X=0,Y=9,Z=0),Scale3D=(X=0.5,Y=1,Z=1))),(Item=(Type=Bone,Name=\"deform_upperarm_fr_r\"),Transform=(Rotation=(X=0,Y=0,Z=0,W=-1),Translation=(X=14.102,Y=-19.0763,Z ...(truncated)"
    Weight=@n6.OutputParam
   )
   &RigUnit_ParentConstraint(
    AdvancedSettings=(RotationOrderForFilter=XZY)
    Child=(Type=Bone,Name="deform_upperarm_bk_base_r")
    Parents=((Item=(Type=Bone,Name="upperRetArm_twist_01_r"),Weight=0.3),(Item=(Type=Bone,Name="clavicle_r"),Weight=0.7))
    Weight=1
    ExecutePin=@RigVMFunction_Sequence_1.C
   )
   > n7 = &RigUnit_SphericalPoseReader(
    ActiveRegionScaleFactors=(PositiveWidth=3,NegativeWidth=1,PositiveHeight=1.5,NegativeHeight=1.5)
    ActiveRegionSize=0.1
    Debug=(bDrawDebug=false,DebugScale=25,DebugSegments=20,DebugThickness=0.25)
    DriverAxis=(X=-1,Y=0,Z=0)
    DriverItem=(Type=Bone,Name="upperarm_r")
    FalloffRegionScaleFactors=(PositiveWidth=1,NegativeWidth=1,PositiveHeight=1,NegativeHeight=1)
    FalloffSize=0.4
    FlipHeightScaling=true
    FlipWidthScaling=false
    OptionalParentItem=(Type=Bone)
    RotationOffset=(X=90,Y=0,Z=0)
   )
   > &RigUnit_ModifyTransforms(
    ItemToModify=((Item=(Type=Bone,Name="deform_upperarm_bk_r"),Transform=(Rotation=(X=0,Y=0,Z=0,W=1),Translation=(X=0,Y=3,Z=0),Scale3D=(X=1,Y=1,Z=1))))
    Weight=@n7.OutputParam
   )
   > RigVMFunctionReturnNode()
  ]}
 CollapseNode_ContainedGraph {nodes [
   n1 = RigVMFunctionEntryNode()
   n2 = &RigVMFunction_Sequence(ExecuteContext=@n1.ExecuteContext)
   RigVMFunction_Sequence = &RigVMFunction_Sequence(ExecuteContext=@n2.A)
   DeformRig_GetDeltaRefTransform_FNC = RigVMCollapseNode(Item=(Type=Bone,Name="lowerRetArm_l"), ExecuteContext=@RigVMFunction_Sequence.A)
   > &RigUnit_ParentConstraint(
    AdvancedSettings=(RotationOrderForFilter=XZY)
    Child=(Type=Bone,Name="deform_elbow_in_base_l")
    Parents=((Item=(Type=Bone,Name="lowerRetArm_l"),Weight=0.4),(Item=(Type=Bone,Name="upperRetArm_l"),Weight=0.6))
    Weight=1
   )
   n3 = &RigVMFunction_MathFloatRemap(SourceMaximum=90, SourceMinimum=15, bClamp=true, Value=@DeformRig_GetDeltaRefTransform_FNC.OutTransform.Rotation.Yaw)
   > &RigUnit_ModifyTransforms(
    ItemToModify=((Item=(Type=Bone,Name="deform_elbow_in_l"),Transform=(Rotation=(X=0,Y=0,Z=0,W=1),Translation=(X=0,Y=10,Z=0),Scale3D=(X=0.5,Y=1,Z=1))))
    Weight=@n3.Result
   )
   &RigUnit_ParentConstraint(
    AdvancedSettings=(RotationOrderForFilter=XZY)
    Child=(Type=Bone,Name="deform_elbow_out_base_l")
    Parents=((Item=(Type=Bone,Name="lowerRetArm_l"),Weight=0.3),(Item=(Type=Bone,Name="upperRetArm_l"),Weight=0.7))
    Weight=1
    ExecutePin=@RigVMFunction_Sequence.B
   )
   > &RigUnit_ModifyTransforms(
    ItemToModify=((Item=(Type=Bone,Name="deform_elbow_out_l"),Transform=(Rotation=(X=0,Y=0,Z=0,W=1),Translation=(X=0,Y=-5,Z=0),Scale3D=(X=0.8,Y=1,Z=1))))
    Weight=@n3.Result
   )
   n4 = &RigVMFunction_Sequence(ExecuteContext=@n2.B)
   &RigUnit_ParentConstraint(
    AdvancedSettings=(RotationOrderForFilter=XZY)
    Child=(Type=Bone,Name="deform_elbow_out_base_r")
    Parents=((Item=(Type=Bone,Name="lowerRetArm_r"),Weight=0.3),(Item=(Type=Bone,Name="upperRetArm_r"),Weight=0.7))
    Weight=1
    ExecutePin=@n4.B
   )
   n5 = &RigVMFunction_MathFloatRemap(SourceMaximum=90, SourceMinimum=15, Value=0, bClamp=true)
   > &RigUnit_ModifyTransforms(
    ItemToModify=((Item=(Type=Bone,Name="deform_elbow_out_r"),Transform=(Rotation=(X=0,Y=0,Z=0,W=1),Translation=(X=0,Y=5,Z=0),Scale3D=(X=0.8,Y=1,Z=1))))
    Weight=@n5.Result
   )
   > RigVMFunctionReturnNode()
   &RigUnit_ParentConstraint(
    AdvancedSettings=(RotationOrderForFilter=XZY)
    Child=(Type=Bone,Name="deform_elbow_in_base_r")
    ExecutePin=()
    Parents=((Item=(Type=Bone,Name="lowerRetArm_r"),Weight=0.4),(Item=(Type=Bone,Name="upperRetArm_r"),Weight=0.6))
    Weight=1
   )
   > &RigUnit_ModifyTransforms(
    ItemToModify=((Item=(Type=Bone,Name="deform_elbow_in_r"),Transform=(Rotation=(X=0,Y=0,Z=0,W=1),Translation=(X=0,Y=-10,Z=0),Scale3D=(X=1,Y=1,Z=1))))
    Weight=@n5.Result
   )
  ]}
 CollapseNode_ContainedGraph {nodes [
   n1 = RigVMFunctionEntryNode()
   RigVMFunction_Sequence = RigVMAggregateNode(ExecuteContext=@n1.Value)
   DeformRig_GetDeltaRefTransform_FNC = RigVMFunctionReferenceNode(Item=(Type=Bone,Name="calf_l"), ExecuteContext=@RigVMFunction_Sequence.A)
   > DeformRig_GetDeltaRefTransform_FNC_1 = RigVMFunctionReferenceNode(Item=(Type=Bone,Name="calf_r"))
   n2 = &RigVMFunction_Sequence(ExecuteContext=@RigVMFunction_Sequence.B)
   n3 = &RigVMFunction_Sequence(ExecuteContext=@n2.A)
   &RigUnit_ParentConstraint(
    AdvancedSettings=(RotationOrderForFilter=XZY)
    Child=(Type=Bone,Name="deform_knee_in_base_l")
    Parents=((Item=(Type=Bone,Name="calfRet_l"),Weight=0.5),(Item=(Type=Bone,Name="thighRet_l"),Weight=0.5))
    Weight=1
    ExecutePin=@n3.A
   )
   n4 = &RigVMFunction_MathFloatRemap(SourceMaximum=100, SourceMinimum=10, bClamp=true, Value=@DeformRig_GetDeltaRefTransform_FNC.OutTransform.Rotation.Yaw)
   > &RigUnit_ModifyTransforms(
    ItemToModify=((Item=(Type=Bone,Name="deform_knee_in_l"),Transform=(Rotation=(X=0,Y=0,Z=0,W=1),Translation=(X=0,Y=-12,Z=0),Scale3D=(X=1,Y=1,Z=1))))
    Weight=@n4.Result
   )
   &RigUnit_ParentConstraint(
    AdvancedSettings=(RotationOrderForFilter=XZY)
    Child=(Type=Bone,Name="deform_knee_in_base_r")
    Parents=((Item=(Type=Bone,Name="calfRet_r"),Weight=0.5),(Item=(Type=Bone,Name="thighRet_r"),Weight=0.5))
    Weight=1
    ExecutePin=@n3.B
   )
   n5 = &RigVMFunction_MathFloatRemap(SourceMaximum=100, SourceMinimum=10, bClamp=true, Value=@DeformRig_GetDeltaRefTransform_FNC_1.OutTransform.Rotation.Yaw)
   > &RigUnit_ModifyTransforms(
    ItemToModify=((Item=(Type=Bone,Name="deform_knee_in_r"),Transform=(Rotation=(X=0,Y=0,Z=0,W=1),Translation=(X=0,Y=12,Z=0),Scale3D=(X=1,Y=1,Z=1))))
    Weight=@n5.Result
   )
   n6 = &RigVMFunction_Sequence(ExecuteContext=@n2.B)
   &RigUnit_ParentConstraint(
    AdvancedSettings=(RotationOrderForFilter=XZY)
    Child=(Type=Bone,Name="deform_knee_out_base_l")
    Parents=((Item=(Type=Bone,Name="calfRet_l"),Weight=0.1),(Item=(Type=Bone,Name="thighRet_l"),Weight=0.9))
    Weight=1
    ExecutePin=@n6.A
   )
   > &RigUnit_ModifyTransforms(
    ItemToModify=((Item=(Type=Bone,Name="deform_knee_out_l"),Transform=(Rotation=(X=0,Y=0,Z=0,W=1),Translation=(X=-10,Y=20,Z=0),Scale3D=(X=1,Y=1,Z=1))))
    Weight=@n4.Result
   )
   &RigUnit_ParentConstraint(
    AdvancedSettings=(RotationOrderForFilter=XZY)
    Child=(Type=Bone,Name="deform_knee_out_base_r")
    Parents=((Item=(Type=Bone,Name="calfRet_r"),Weight=0.1),(Item=(Type=Bone,Name="thighRet_r"),Weight=0.9))
    Weight=1
    ExecutePin=@n6.B
   )
   > &RigUnit_ModifyTransforms(
    ItemToModify=((Item=(Type=Bone,Name="deform_knee_out_r"),Transform=(Rotation=(X=0,Y=0,Z=0,W=1),Translation=(X=20,Y=-5,Z=0),Scale3D=(X=1,Y=1,Z=1))))
    Weight=@n5.Result
   )
   RigVMFunctionReturnNode(Value=@RigVMFunction_Sequence.C)
  ]}
 CollapseNode_ContainedGraph {nodes [
   n1 = RigVMFunctionEntryNode()
   RigVMFunction_Sequence = RigVMAggregateNode(ExecuteContext=@n1.ExecutePin)
   n2 = &RigUnit_SphericalPoseReader(
    ActiveRegionScaleFactors=(PositiveWidth=3.5,NegativeWidth=0.5,PositiveHeight=3,NegativeHeight=3)
    ActiveRegionSize=0.1
    Debug=(bDrawDebug=false,DebugScale=25,DebugSegments=20,DebugThickness=0.25)
    DriverAxis=(X=-1,Y=0,Z=0)
    DriverItem=(Type=Bone,Name="thigh_l")
    FalloffRegionScaleFactors=(PositiveWidth=1,NegativeWidth=1,PositiveHeight=1,NegativeHeight=1)
    FalloffSize=0.45
    OptionalParentItem=(Type=Bone)
    RotationOffset=(X=90,Y=0,Z=0)
    ExecutePin=@RigVMFunction_Sequence.A
   )
   > n3 = &RigUnit_SphericalPoseReader(
    ActiveRegionScaleFactors=(PositiveWidth=3.5,NegativeWidth=0.5,PositiveHeight=3,NegativeHeight=3)
    ActiveRegionSize=0.1
    Debug=(bDrawDebug=false,DebugScale=25,DebugSegments=20,DebugThickness=0.25)
    DriverAxis=(X=1,Y=0,Z=0)
    DriverItem=(Type=Bone,Name="thigh_r")
    FalloffRegionScaleFactors=(PositiveWidth=1,NegativeWidth=1,PositiveHeight=1,NegativeHeight=1)
    FalloffSize=0.45
    FlipHeightScaling=false
    FlipWidthScaling=true
    OptionalParentItem=(Type=Bone)
    RotationOffset=(X=-90,Y=0,Z=0)
   )
   n4 = &RigVMFunction_Sequence(ExecuteContext=@RigVMFunction_Sequence.B)
   n5 = &RigVMFunction_Sequence(ExecuteContext=@n4.B)
   &RigUnit_ParentConstraint(
    AdvancedSettings=(RotationOrderForFilter=XZY)
    Child=(Type=Bone,Name="deform_glute_base_l")
    Parents=((Item=(Type=Bone,Name="thighRet_l"),Weight=0.25),(Item=(Type=Bone,Name="pelvis"),Weight=0.75))
    Weight=1
    ExecutePin=@n5.A
   )
   > &RigUnit_ModifyTransforms(
    ItemToModify=((Item=(Type=Bone,Name="deform_glute_l"),Transform=(Rotation=(X=0,Y=0,Z=0,W=1),Translation=(X=-7,Y=0,Z=0),Scale3D=(X=1,Y=1,Z=1))))
    Weight=@n2.OutputParam
   )
   &RigUnit_ParentConstraint(
    AdvancedSettings=(RotationOrderForFilter=XZY)
    Child=(Type=Bone,Name="deform_glute_base_r")
    Parents=((Item=(Type=Bone,Name="thighRet_r"),Weight=0.25),(Item=(Type=Bone,Name="pelvis"),Weight=0.75))
    Weight=1
    ExecutePin=@n5.B
   )
   > &RigUnit_ModifyTransforms(
    ItemToModify=((Item=(Type=Bone,Name="deform_glute_r"),Transform=(Rotation=(X=0,Y=0,Z=0,W=1),Translation=(X=7,Y=0,Z=0),Scale3D=(X=1,Y=1,Z=1))))
    Weight=@n3.OutputParam
   )
   n6 = &RigVMFunction_Sequence(ExecuteContext=@n4.A)
   &RigUnit_ParentConstraint(
    AdvancedSettings=(RotationOrderForFilter=XZY)
    Child=(Type=Bone,Name="deform_groin_base_l")
    Parents=((Item=(Type=Bone,Name="thighRet_l"),Weight=0.7),(Item=(Type=Bone,Name="pelvis"),Weight=0.3))
    Weight=1
    ExecutePin=@n6.A
   )
   > &RigUnit_ModifyTransforms(
    ItemToModify=((Item=(Type=Bone,Name="deform_groin_l"),Transform=(Rotation=(X=0,Y=0,Z=0,W=1),Translation=(X=-3,Y=12,Z=0),Scale3D=(X=1,Y=1,Z=1))))
    Weight=@n2.OutputParam
   )
   &RigUnit_ParentConstraint(
    AdvancedSettings=(RotationOrderForFilter=XZY)
    Child=(Type=Bone,Name="deform_groin_base_r")
    Parents=((Item=(Type=Bone,Name="thighRet_r"),Weight=0.7),(Item=(Type=Bone,Name="pelvis"),Weight=0.3))
    Weight=1
    ExecutePin=@n6.B
   )
   > &RigUnit_ModifyTransforms(
    ItemToModify=((Item=(Type=Bone,Name="deform_groin_r"),Transform=(Rotation=(X=0,Y=0,Z=0,W=1),Translation=(X=3,Y=-12,Z=0),Scale3D=(X=1,Y=1,Z=1))))
    Weight=@n3.OutputParam
   )
   RigVMFunctionReturnNode(ExecutePin=@RigVMFunction_Sequence.C)
  ]}
 ContainedGraph {nodes [
   n1 = RigVMFunctionEntryNode()
   n2 = &RigUnit_GetTransform(Space=LocalSpace, Item=@n1.Item)
   n3 = &RigUnit_GetTransform(Space=LocalSpace, bInitial=true, Item=@n1.Item)
   > RigVMFunctionReturnNode(OutTransform=/Script/RigVM.RigVMFunction_MathTransformToEulerTransform(Value=&RigVMFunction_MathTransformMakeRelative(Global=@n2.Transform, Parent=@n3.Transform).Local) {is=struct RigVMUnitNode, template="Cast::Execute(in Value,out Result)"}.Result)
  ]}
 ContainedGraph {nodes [
   n1 = RigVMFunctionEntryNode()
   n2 = &RigVMFunction_MathTransformToEulerTransform(Value=@n1.transform) {template="Cast::Execute(in Value,out Result)"}
   > RigVMFunctionReturnNode(
    mirrorTransform=(Rotation=(X=0,Y=0,Z=0,W=1),Translation=(X=0,Y=0,Z=0),Scale3D=(X=0,Y=0,Z=0))
    mirrorTransform.Rotation=/Script/RigVM.RigVMFunction_MathQuaternionFromRotatorV2(Value=@n2.Result.Rotation) {is=struct RigVMUnitNode, template="Cast::Execute(in Value,out Result)"}.Result
    mirrorTransform.Translation=/Script/RigVM.RigVMFunction_MathVectorScale(Factor=-1, Value=@n2.Result.Location) {is=struct RigVMUnitNode, template="Scale::Execute(in Value,in Factor,out Result)"}.Result
    mirrorTransform.Scale3D=@n2.Result.Scale
   )
  ]}
 ContainedGraph {nodes [
   n1 = RigVMFunctionEntryNode()
   n2 = &RigUnit_GetTransform(Space=LocalSpace, Item=@n1.Item)
   n3 = &RigUnit_GetTransform(Space=LocalSpace, bInitial=true, Item=@n1.Item)
   > RigVMFunctionReturnNode(OutTransform=&RigVMFunction_MathTransformToEulerTransform(Value=&RigVMFunction_MathTransformMakeRelative(Global=@n2.Transform, Parent=@n3.Transform).Local) {template="Cast::Execute(in Value,out Result)"}.Result)
  ]}
 New Function_ContainedGraph {nodes [
   n1 = RigVMFunctionEntryNode()
   RigVMFunction_MathBoolToInteger = &RigVMFunction_MathBoolToInteger(Value=@n1.onlyTwist)
   RigVMFunction_MathQuaternionSwingTwist = &RigVMFunction_MathQuaternionSwingTwist(
    twistAxis=@n1.twistAxis
    Input=&RigUnit_GetRelativeTransformForItem(bChildInitial=False, bParentInitial=False, child=@n1.child, parent=@n1.parent).RelativeTransform.Rotation
   )
   n2 = RigVMDispatchNode(
    Values=((),())
    Index=@RigVMFunction_MathBoolToInteger.Result
    Values.0=@RigVMFunction_MathQuaternionSwingTwist.Swing
    Values.1=@RigVMFunction_MathQuaternionSwingTwist.Twist
   ) {template="DISPATCH_RigVMDispatch_SelectInt32(in Index,in Values,out Result)"}
   Inverse = /Script/RigVM.RigVMFunction_MathQuaternionInverse(Value=@n2.Result) {is=struct RigVMUnitNode, template="Inverse::Execute(in Value,out Result)"}
   > RigVMFunctionReturnNode(outputQuat=RigVMDispatchNode(True=@Inverse.Result, Condition=@n1.invert, False=@n2.Result) {template="DISPATCH_RigVMDispatch_If(in Condition,in True,in False,out Result)"}.Result)
  ]}
 New Function_ContainedGraph {nodes [
   n1 = RigVMFunctionEntryNode()
   RigVMFunction_ControlFlowBranch = &RigVMFunction_ControlFlowBranch(ExecuteContext=@n1.ExecuteContext, Condition=@n1.enable)
   RigVMFunctionReturnNode(ExecuteContext=@RigVMFunction_ControlFlowBranch.Completed)
   For_Each = RigVMDispatchNode(Array=@n1.chainArray, ExecuteContext=@RigVMFunction_ControlFlowBranch.True) {template="DISPATCH_RigVMDispatch_ArrayIterator(in Array,out Element,out Index,out Count,out Ratio)"}
   Interpolate = /Script/RigVM.RigVMFunction_MathQuaternionSlerp(T=@For_Each.Ratio, B=@n1.botRotQuat, A=@n1.topRotQuat) {is=struct RigVMUnitNode, template="Interpolate::Execute(in A,in B,in T,out Result)"}
   > &RigUnit_OffsetTransformForItem(
    OffsetTransform=(Rotation=(X=0,Y=0,Z=0,W=1),Translation=(X=0,Y=0,Z=0),Scale3D=(X=1,Y=1,Z=1))
    Weight=1
    bPropagateToChildren=false
    OffsetTransform.Rotation=@Interpolate.Result
    Item=@For_Each.Element
   )
  ]}
 New Function_ContainedGraph {nodes [
   n1 = RigVMFunctionEntryNode()
   n2 = &RigVMFunction_ControlFlowBranch(ExecuteContext=@n1.ExecuteContext, Condition=RigVMVariableNode() {variable=lockPV}.Value)
   RigVMFunctionReturnNode(ExecuteContext=@n2.Completed)
   n3 = &RigUnit_GetTransform(bInitial=true, Item=RigVMVariableNode() {variable=secondBone}.Value)
   n4 = &RigVMFunction_MathVectorDistance(
    B=@n3.Transform.Translation
    A=&RigUnit_GetTransform(bInitial=true, Item=RigVMVariableNode() {variable=firstBone}.Value).Transform.Translation
   )
   n5 = &RigVMFunction_MathVectorDistance(
    A=@n3.Transform.Translation
    B=&RigUnit_GetTransform(bInitial=true, Item=RigVMVariableNode() {variable=effectorBone}.Value).Transform.Translation
   )
   n6 = &RigVMFunction_MathDoubleAdd(A=@n4.Result, B=@n5.Result)
   RigVMVariableNode(Value=@n6.Result, ExecuteContext=@n2.False) {variable=Entry_segmentsSum}
   > RigVMVariableNode(Value=&RigVMFunction_MathDoubleDiv(B=@n6.Result, A=@n4.Result).Result) {variable=Entry_segmentAFactor}
   > RigVMVariableNode(Value=&RigVMFunction_MathDoubleDiv(B=@n6.Result, A=@n5.Result).Result) {variable=Entry_segmentBFactor}
   > &RigUnit_ParentConstraint(
    AdvancedSettings=(InterpolationType=Average,RotationOrderForFilter=XZY)
    Parents=((Item=(Type=Bone,Name="None"),Weight=1))
    Weight=1
    bMaintainOffset=true
    Parents.0.Item=RigVMVariableNode() {variable=firstBoneOrig}.Value
    Child=RigVMVariableNode() {variable=poleVector}.Value
   )
   > &RigUnit_ModifyTransforms(
    ItemToModify=((Item=(Type=Bone,Name="None"),Transform=(Rotation=(X=0,Y=0,Z=0,W=1),Translation=(X=0,Y=0,Z=0),Scale3D=(X=1,Y=1,Z=1))))
    Weight=1
    WeightMinimum=0
    ItemToModify.0.Transform.Translation=RigVMVariableNode() {variable=offsetPv}.Value
    ItemToModify.0.Item=RigVMVariableNode() {variable=poleVector}.Value
   )
   RigVMFunction_MathVectorDistance = &RigVMFunction_MathVectorDistance(
    B=&RigUnit_GetTransform(bInitial=False, Item=RigVMVariableNode() {variable=ikEffector}.Value).Transform.Translation
    A=&RigUnit_GetTransform(bInitial=False, Item=RigVMVariableNode() {variable=itemStartStretch}.Value).Transform.Translation
   )
   n7 = RigVMVariableNode() {variable=softness}
   n8 = &RigVMFunction_MathDoubleSub(B=@n7.Value, A=RigVMVariableNode() {variable=Entry_segmentsSum}.Value)
   Divide = &RigVMFunction_MathDoubleDiv(
    A=&RigVMFunction_MathDoubleMul(B=-1, A=&RigVMFunction_MathDoubleSub(A=@RigVMFunction_MathVectorDistance.Result, B=@n8.Result).Result).Result
    B=@n7.Value
   )
   Exponential = /Script/RigVM.RigVMFunction_MathDoubleExponential(Value=@Divide.Result) {is=struct RigVMUnitNode, template="Exponential::Execute(in Value,out Result)"}
   Add = &RigVMFunction_MathDoubleAdd(
    B=RigVMDispatchNode(
     False=0
     Condition=/Script/RigVM.RigVMFunction_MathDoubleGreater(A=@RigVMFunction_MathVectorDistance.Result, B=@n8.Result) {is=struct RigVMUnitNode, template="Greater::Execute(in A,in B,out Result)"}.Result
     True=&RigVMFunction_MathDoubleMul(A=@Exponential.Result, B=@n7.Value).Result
    ) {template="DISPATCH_RigVMDispatch_If(in Condition,in True,in False,out Result)"}.Result
    A=@RigVMFunction_MathVectorDistance.Result
   )
   Greater = &RigVMFunction_MathDoubleGreater(A=@Add.Result, B=RigVMVariableNode() {variable=Entry_segmentsSum}.Value) {template="Greater::Execute(in A,in B,out Result)"}
   n9 = RigVMDispatchNode(True=@Add.Result, Condition=@Greater.Result, False=RigVMVariableNode() {variable=Entry_segmentsSum}.Value) {template="DISPATCH_RigVMDispatch_If(in Condition,in True,in False,out Result)"}
   Multiply = &RigVMFunction_MathDoubleMul(B=@n9.Result, A=RigVMVariableNode() {variable=Entry_segmentAFactor}.Value)
   > /Script/ControlRig.RigUnit_TwoBoneIKSimplePerItem(
    DebugSettings=(bEnabled=False,Scale=10,WorldOffset=(Rotation=(X=0,Y=0,Z=0,W=1),Translation=(X=0,Y=0,Z=0),Scale3D=(X=1,Y=1,Z=1)))
    PoleVectorKind=Location
    PoleVectorSpace=(Type=Bone,Name="None")
    SecondaryAxisWeight=1
    StretchMaximumRatio=2
    StretchStartRatio=1
    Weight=1
    bEnableStretch=false
    bPropagateToChildren=True
    poleVector=&RigUnit_GetTransform(bInitial=False, Item=RigVMVariableNode() {variable=poleVector}.Value).Transform.Translation
    primaryAxis=RigVMVariableNode() {variable=primaryAxis}.Value
    secondaryAxis=RigVMVariableNode() {variable=secondaryAxis}.Value
    ItemA=RigVMVariableNode() {variable=firstBone}.Value
    ItemB=RigVMVariableNode() {variable=secondBone}.Value
    EffectorItem=RigVMVariableNode() {variable=effectorBone}.Value
    Effector=&RigUnit_GetTransform(bInitial=False, Item=RigVMVariableNode() {variable=ikEffector}.Value).Transform
    ItemALength=@Multiply.Result
    ItemBLength=&RigVMFunction_MathDoubleMul(A=@n9.Result, B=RigVMVariableNode() {variable=Entry_segmentBFactor}.Value).Result
   ) {is=struct RigVMUnitNode}
   > n10 = &RigUnit_ParentConstraint(
    AdvancedSettings=(InterpolationType=Average,RotationOrderForFilter=XZY)
    Parents=((Item=(Type=Bone,Name="None"),Weight=1))
    Weight=1
    Child=RigVMVariableNode() {variable=effectorBone}.Value
    Parents.0.Item=RigVMVariableNode() {variable=secondBone}.Value
   )
   RigVMFunction_ControlFlowBranch = &RigVMFunction_ControlFlowBranch(ExecuteContext=@n10.ExecutePin, Condition=RigVMVariableNode() {variable=secondBoneConstraint}.Value)
   n11 = RigVMVariableNode() {variable=bendAxis}
   &RigUnit_AimConstraintLocalSpaceOffset(
    Filter=(bX=True,bY=True,bZ=True)
    Parents=((Item=(Type=Bone,Name="None"),Weight=1))
    Weight=1
    WorldUp=(Target=(X=0,Y=0,Z=1),Kind=Direction,Space=(Type=None,Name="None"))
    bMaintainOffset=false
    UpAxis=@n11.Value
    WorldUp.Target=@n11.Value
    Child=RigVMVariableNode() {variable=secondBone}.Value
    AimAxis=RigVMVariableNode() {variable=primaryAxis}.Value
    WorldUp.Space=RigVMVariableNode() {variable=firstBone}.Value
    Parents.0.Item=RigVMVariableNode() {variable=ikEffector}.Value
    ExecutePin=@RigVMFunction_ControlFlowBranch.True
   )
   &RigUnit_ParentConstraint(
    AdvancedSettings=(InterpolationType=Average,RotationOrderForFilter=XZY)
    Parents=((Item=(Type=Bone,Name="None"),Weight=1))
    Weight=1
    bMaintainOffset=false
    Child=RigVMVariableNode() {variable=poleVector}.Value
    Parents.0.Item=RigVMVariableNode() {variable=secondBoneOrig}.Value
    ExecutePin=@n2.True
   )
   > &RigUnit_ModifyTransforms(
    ItemToModify=((Item=(Type=Bone,Name="None"),Transform=(Rotation=(X=0,Y=0,Z=0,W=1),Translation=(X=0,Y=0,Z=0),Scale3D=(X=1,Y=1,Z=1))))
    Weight=1
    WeightMinimum=0
    ItemToModify.0.Transform.Translation=RigVMVariableNode() {variable=offsetPv}.Value
    ItemToModify.0.Item=RigVMVariableNode() {variable=poleVector}.Value
   )
   n12 = RigVMVariableNode() {variable=secondaryAxis}
   > &RigUnit_AimConstraintLocalSpaceOffset(
    Filter=(bX=True,bY=True,bZ=True)
    Parents=((Item=(Type=Bone,Name="None"),Weight=1))
    Weight=1
    WorldUp=(Target=(X=0,Y=1,Z=0),Kind=Direction,Space=(Type=None,Name="None"))
    bMaintainOffset=false
    AimAxis=RigVMVariableNode() {variable=primaryAxis}.Value
    UpAxis=@n12.Value
    WorldUp.Target=@n12.Value
    WorldUp.Space=RigVMVariableNode() {variable=firstBoneOrig}.Value
    Parents.0.Item=RigVMVariableNode() {variable=poleVector}.Value
    Child=RigVMVariableNode() {variable=firstBone}.Value
   )
   > /Script/ControlRig.RigUnit_PositionConstraintLocalSpaceOffset(
    Filter=(bX=True,bY=True,bZ=True)
    Parents=((Item=(Type=Bone,Name="None"),Weight=1))
    Weight=1
    bMaintainOffset=false
    Parents.0.Item=RigVMVariableNode() {variable=poleVector}.Value
    Child=RigVMVariableNode() {variable=secondBone}.Value
   ) {is=struct RigVMUnitNode}
   n13 = RigVMVariableNode() {variable=secondaryAxis}
   > &RigUnit_AimConstraintLocalSpaceOffset(
    Filter=(bX=True,bY=True,bZ=True)
    Parents=((Item=(Type=Bone,Name="None"),Weight=1))
    Weight=1
    WorldUp=(Target=(X=0,Y=1,Z=0),Kind=Direction,Space=(Type=None,Name="None"))
    bMaintainOffset=false
    Child=RigVMVariableNode() {variable=secondBone}.Value
    AimAxis=RigVMVariableNode() {variable=primaryAxis}.Value
    UpAxis=@n13.Value
    WorldUp.Target=@n13.Value
    WorldUp.Space=RigVMVariableNode() {variable=secondBoneOrig}.Value
    Parents.0.Item=RigVMVariableNode() {variable=ikEffector}.Value
   )
  ]}
 New Function_ContainedGraph {nodes [
   n1 = RigVMFunctionEntryNode()
   n2 = RigVMVariableNode() {variable=isWearingGauntletsLeft}
   AlphaInterp = &RigVMFunction_AlphaInterp(
    ClampMax=10
    ClampMin=5
    InRange=(Min=5,Max=10)
    InterpSpeedDecreasing=11
    InterpSpeedIncreasing=11
    OutRange=(Min=5,Max=10)
    Scale=1
    bClampResult=true
    bMapRange=true
    Value=RigVMVariableNode() {variable=ArmsIK_Softness}.Value
   )
   TwoBoneIKWithSoftness v4_1 = RigVMFunctionReferenceNode(
    bendAxis=(X=0,Y=0,Z=1)
    effectorBone=(Type=Bone,Name="handRet_l")
    firstBone=(Type=Bone,Name="upperRetArm_l")
    firstBoneOrig=(Type=Bone,Name="upperarm_l")
    ikEffector=(Type=Bone,Name="hand_ik_l")
    itemStartStretch=(Type=Null,Name="upperRetArm_l_null")
    poleVector=(Type=Control,Name="arm_l_pv_ctrl")
    primaryAxis=(X=1,Y=0,Z=0)
    secondBone=(Type=Bone,Name="lowerRetArm_l")
    secondBoneConstraint=true
    secondBoneOrig=(Type=Bone,Name="lowerarm_l")
    secondaryAxis=(X=0,Y=-1,Z=0)
    ExecuteContext=@n1.Argument
    offsetPv=RigVMDispatchNode(False=(X=0,Y=0,Z=30), True=(X=0,Y=0,Z=0), Condition=@n2.Value) {template="DISPATCH_RigVMDispatch_If(in Condition,in True,in False,out Result)"}.Result
    softness=@AlphaInterp.Result
    lockPV=@n2.Value
   )
   n3 = RigVMVariableNode() {variable=isWearingGauntletsRight}
   > TwoBoneIKWithSoftness v4_3 = RigVMFunctionReferenceNode(
    bendAxis=(X=0,Y=0,Z=1)
    effectorBone=(Type=Bone,Name="handRet_r")
    firstBone=(Type=Bone,Name="upperRetArm_r")
    firstBoneOrig=(Type=Bone,Name="upperarm_r")
    ikEffector=(Type=Bone,Name="hand_ik_r")
    itemStartStretch=(Type=Null,Name="upperRetArm_r_null")
    poleVector=(Type=Control,Name="arm_r_pv_ctrl")
    primaryAxis=(X=-1,Y=0,Z=0)
    secondBone=(Type=Bone,Name="lowerRetArm_r")
    secondBoneConstraint=true
    secondBoneOrig=(Type=Bone,Name="lowerarm_r")
    secondaryAxis=(X=0,Y=1,Z=0)
    offsetPv=RigVMDispatchNode(False=(X=0,Y=0,Z=30), True=(X=0,Y=0,Z=0), Condition=@n3.Value) {template="DISPATCH_RigVMDispatch_If(in Condition,in True,in False,out Result)"}.Result
    softness=@AlphaInterp.Result
    lockPV=@n3.Value
   )
   > RigVMFunction_Sequence_2 = RigVMAggregateNode()
   DeltaQuatExtraction = RigVMFunctionReferenceNode(
    child=(Type=Bone,Name="upperRetArm_l")
    invert=true
    onlyTwist=true
    parent=(Type=Null,Name="upperRetArm_l_null")
    twistAxis=(X=1,Y=0,Z=0)
    ExecuteContext=@RigVMFunction_Sequence_2.A
   )
   ItemArray = RigVMDispatchNode(Value=((Type=Bone,Name="upperRetArm_twist_01_l"),(Type=Bone,Name="upperRetArm_twist_02_l"),(Type=Bone,Name="upperRetArm_twist_03_l"))) {template="DISPATCH_RigVMDispatch_Constant(io Value)"}
   > SplineTwistTopBot = RigVMFunctionReferenceNode(botRotQuat=(), enable=true, topRotQuat=@DeltaQuatExtraction.outputQuat, chainArray=@ItemArray.Value)
   > DeltaQuatExtraction_1 = RigVMFunctionReferenceNode(child=(Type=Bone,Name="hand_l"), invert=false, onlyTwist=true, parent=(Type=Null,Name="hand_l_null"), twistAxis=(X=1,Y=0,Z=0))
   ItemArray_2 = RigVMDispatchNode(Value=((Type=None,Name="None"),(Type=Bone,Name="lowerRetArm_twist_01_l"),(Type=Bone,Name="lowerRetArm_twist_02_l"),(Type=Bone,Name="lowerRetArm_twist_03_l"),(Type=Bone,Name="lowerRetArm_twist_04_l"))) {template="DISPATCH_RigVMDispatch_Constant(io Value)"}
   > SplineTwistTopBot_1 = RigVMFunctionReferenceNode(enable=true, topRotQuat=(), botRotQuat=@DeltaQuatExtraction_1.outputQuat, chainArray=@ItemArray_2.Value)
   DeltaQuatExtraction_2 = RigVMFunctionReferenceNode(
    child=(Type=Bone,Name="upperRetArm_r")
    invert=true
    onlyTwist=true
    parent=(Type=Null,Name="upperRetArm_r_null")
    twistAxis=(X=1,Y=0,Z=0)
    ExecuteContext=@RigVMFunction_Sequence_2.B
   )
   ItemArray_1 = RigVMDispatchNode(Value=((Type=Bone,Name="upperRetArm_twist_01_r"),(Type=Bone,Name="upperRetArm_twist_02_r"),(Type=Bone,Name="upperRetArm_twist_03_r"))) {template="DISPATCH_RigVMDispatch_Constant(io Value)"}
   > SplineTwistTopBot_3 = RigVMFunctionReferenceNode(botRotQuat=(), enable=true, topRotQuat=@DeltaQuatExtraction_2.outputQuat, chainArray=@ItemArray_1.Value)
   > DeltaQuatExtraction_3 = RigVMFunctionReferenceNode(child=(Type=Bone,Name="hand_r"), invert=false, onlyTwist=true, parent=(Type=Null,Name="hand_r_null"), twistAxis=(X=1,Y=0,Z=0))
   ItemArray_3 = RigVMDispatchNode(Value=((Type=None,Name="None"),(Type=Bone,Name="lowerRetArm_twist_01_r"),(Type=Bone,Name="lowerRetArm_twist_02_r"),(Type=Bone,Name="lowerRetArm_twist_03_r"),(Type=Bone,Name="lowerRetArm_twist_04_r"))) {template="DISPATCH_RigVMDispatch_Constant(io Value)"}
   > SplineTwistTopBot_2 = RigVMFunctionReferenceNode(enable=true, topRotQuat=(), botRotQuat=@DeltaQuatExtraction_3.outputQuat, chainArray=@ItemArray_3.Value)
   RigVMFunction_Sequence = &RigVMFunction_Sequence(ExecuteContext=@RigVMFunction_Sequence_2.C)
   "Upperarm Corrective " = RigVMCollapseNode(ExecuteContext=@RigVMFunction_Sequence.A)
   > &RigUnit_RotationConstraintLocalSpaceOffset(
    AdvancedSettings=(InterpolationType=Average,RotationOrderForFilter=XZY)
    Child=(Type=Bone,Name="deform_ring_01_lf")
    Filter=(bX=True,bY=True,bZ=True)
    Parents=((Item=(Type=Bone,Name="middle_01_l"),Weight=1))
    Weight=1
   )
   > &RigUnit_RotationConstraintLocalSpaceOffset(
    AdvancedSettings=(InterpolationType=Average,RotationOrderForFilter=XZY)
    Child=(Type=Bone,Name="deform_ring_02_lf")
    Filter=(bX=True,bY=True,bZ=True)
    Parents=((Item=(Type=Bone,Name="middle_02_l"),Weight=1))
    Weight=1
   )
   > &RigUnit_RotationConstraintLocalSpaceOffset(
    AdvancedSettings=(InterpolationType=Average,RotationOrderForFilter=XZY)
    Child=(Type=Bone,Name="deform_ring_03_lf")
    Filter=(bX=True,bY=True,bZ=True)
    Parents=((Item=(Type=Bone,Name="middle_03_l"),Weight=1))
    Weight=1
   )
   > &RigUnit_RotationConstraintLocalSpaceOffset(
    AdvancedSettings=(InterpolationType=Average,RotationOrderForFilter=XZY)
    Child=(Type=Bone,Name="deform_ring_01_rt")
    Filter=(bX=True,bY=True,bZ=True)
    Parents=((Item=(Type=Bone,Name="middle_01_r"),Weight=1))
    Weight=1
   )
   > &RigUnit_RotationConstraintLocalSpaceOffset(
    AdvancedSettings=(InterpolationType=Average,RotationOrderForFilter=XZY)
    Child=(Type=Bone,Name="deform_ring_02_rt")
    Filter=(bX=True,bY=True,bZ=True)
    Parents=((Item=(Type=Bone,Name="middle_02_r"),Weight=1))
    Weight=1
   )
   > &RigUnit_RotationConstraintLocalSpaceOffset(
    AdvancedSettings=(InterpolationType=Average,RotationOrderForFilter=XZY)
    Child=(Type=Bone,Name="deform_ring_03_rt")
    Filter=(bX=True,bY=True,bZ=True)
    Parents=((Item=(Type=Bone,Name="middle_03_r"),Weight=1))
    Weight=1
   )
   Elbow Corrective = RigVMCollapseNode(ExecuteContext=@RigVMFunction_Sequence.B)
   RigVMFunctionReturnNode(Argument=@RigVMFunction_Sequence_2.D)
  ]}
 New Function_ContainedGraph {nodes [
   n1 = RigVMFunctionEntryNode()
   n2 = &RigVMFunction_ControlFlowBranch(ExecuteContext=@n1.ExecuteContext, Condition=RigVMVariableNode() {variable=lockPV}.Value)
   RigVMFunctionReturnNode(ExecuteContext=@n2.Completed)
   n3 = &RigUnit_GetTransform(bInitial=true, Item=RigVMVariableNode() {variable=secondBone}.Value)
   n4 = &RigVMFunction_MathVectorDistance(
    B=@n3.Transform.Translation
    A=&RigUnit_GetTransform(bInitial=true, Item=RigVMVariableNode() {variable=firstBone}.Value).Transform.Translation
   )
   n5 = &RigVMFunction_MathVectorDistance(
    A=@n3.Transform.Translation
    B=&RigUnit_GetTransform(bInitial=true, Item=RigVMVariableNode() {variable=effectorBone}.Value).Transform.Translation
   )
   n6 = &RigVMFunction_MathDoubleAdd(A=@n4.Result, B=@n5.Result)
   RigVMVariableNode(Value=@n6.Result, ExecuteContext=@n2.False) {variable=Entry_segmentsSum}
   > RigVMVariableNode(Value=&RigVMFunction_MathDoubleDiv(B=@n6.Result, A=@n4.Result).Result) {variable=Entry_segmentAFactor}
   > RigVMVariableNode(Value=&RigVMFunction_MathDoubleDiv(B=@n6.Result, A=@n5.Result).Result) {variable=Entry_segmentBFactor}
   > &RigUnit_ParentConstraint(
    AdvancedSettings=(InterpolationType=Average,RotationOrderForFilter=XZY)
    Parents=((Item=(Type=Bone,Name="None"),Weight=1))
    Weight=1
    bMaintainOffset=true
    Parents.0.Item=RigVMVariableNode() {variable=firstBoneOrig}.Value
    Child=RigVMVariableNode() {variable=poleVector}.Value
   )
   > &RigUnit_ModifyTransforms(
    ItemToModify=((Item=(Type=Bone,Name="None"),Transform=(Rotation=(X=0,Y=0,Z=0,W=1),Translation=(X=0,Y=0,Z=0),Scale3D=(X=1,Y=1,Z=1))))
    Weight=1
    WeightMinimum=0
    ItemToModify.0.Transform.Translation=RigVMVariableNode() {variable=offsetPv}.Value
    ItemToModify.0.Item=RigVMVariableNode() {variable=poleVector}.Value
   )
   RigVMFunction_MathVectorDistance = &RigVMFunction_MathVectorDistance(
    B=&RigUnit_GetTransform(bInitial=False, Item=RigVMVariableNode() {variable=ikEffector}.Value).Transform.Translation
    A=&RigUnit_GetTransform(bInitial=False, Item=RigVMVariableNode() {variable=itemStartStretch}.Value).Transform.Translation
   )
   n7 = RigVMVariableNode() {variable=softness}
   n8 = &RigVMFunction_MathDoubleSub(B=@n7.Value, A=RigVMVariableNode() {variable=Entry_segmentsSum}.Value)
   Divide = &RigVMFunction_MathDoubleDiv(
    A=&RigVMFunction_MathDoubleMul(B=-1, A=&RigVMFunction_MathDoubleSub(A=@RigVMFunction_MathVectorDistance.Result, B=@n8.Result).Result).Result
    B=@n7.Value
   )
   Exponential = &RigVMFunction_MathDoubleExponential(Value=@Divide.Result) {template="Exponential::Execute(in Value,out Result)"}
   Add = &RigVMFunction_MathDoubleAdd(
    B=RigVMDispatchNode(
     False=0
     Condition=&RigVMFunction_MathDoubleGreater(A=@RigVMFunction_MathVectorDistance.Result, B=@n8.Result) {template="Greater::Execute(in A,in B,out Result)"}.Result
     True=&RigVMFunction_MathDoubleMul(A=@Exponential.Result, B=@n7.Value).Result
    ) {template="DISPATCH_RigVMDispatch_If(in Condition,in True,in False,out Result)"}.Result
    A=@RigVMFunction_MathVectorDistance.Result
   )
   Greater = &RigVMFunction_MathDoubleGreater(A=@Add.Result, B=RigVMVariableNode() {variable=Entry_segmentsSum}.Value) {template="Greater::Execute(in A,in B,out Result)"}
   n9 = RigVMDispatchNode(True=@Add.Result, Condition=@Greater.Result, False=RigVMVariableNode() {variable=Entry_segmentsSum}.Value) {template="DISPATCH_RigVMDispatch_If(in Condition,in True,in False,out Result)"}
   Multiply = &RigVMFunction_MathDoubleMul(B=@n9.Result, A=RigVMVariableNode() {variable=Entry_segmentAFactor}.Value)
   > &RigUnit_TwoBoneIKSimplePerItem(
    DebugSettings=(bEnabled=False,Scale=10,WorldOffset=(Rotation=(X=0,Y=0,Z=0,W=1),Translation=(X=0,Y=0,Z=0),Scale3D=(X=1,Y=1,Z=1)))
    PoleVectorKind=Location
    PoleVectorSpace=(Type=Bone,Name="None")
    SecondaryAxisWeight=1
    StretchMaximumRatio=2
    StretchStartRatio=1
    Weight=1
    bEnableStretch=false
    bPropagateToChildren=True
    poleVector=&RigUnit_GetTransform(bInitial=False, Item=RigVMVariableNode() {variable=poleVector}.Value).Transform.Translation
    primaryAxis=RigVMVariableNode() {variable=primaryAxis}.Value
    secondaryAxis=RigVMVariableNode() {variable=secondaryAxis}.Value
    ItemA=RigVMVariableNode() {variable=firstBone}.Value
    ItemB=RigVMVariableNode() {variable=secondBone}.Value
    EffectorItem=RigVMVariableNode() {variable=effectorBone}.Value
    Effector=&RigUnit_GetTransform(bInitial=False, Item=RigVMVariableNode() {variable=ikEffector}.Value).Transform
    ItemALength=@Multiply.Result
    ItemBLength=&RigVMFunction_MathDoubleMul(A=@n9.Result, B=RigVMVariableNode() {variable=Entry_segmentBFactor}.Value).Result
   )
   > n10 = &RigUnit_ParentConstraint(
    AdvancedSettings=(InterpolationType=Average,RotationOrderForFilter=XZY)
    Parents=((Item=(Type=Bone,Name="None"),Weight=1))
    Weight=1
    Child=RigVMVariableNode() {variable=effectorBone}.Value
    Parents.0.Item=RigVMVariableNode() {variable=secondBone}.Value
   )
   RigVMFunction_ControlFlowBranch = &RigVMFunction_ControlFlowBranch(ExecuteContext=@n10.ExecutePin, Condition=RigVMVariableNode() {variable=secondBoneConstraint}.Value)
   n11 = RigVMVariableNode() {variable=bendAxis}
   &RigUnit_AimConstraintLocalSpaceOffset(
    Filter=(bX=True,bY=True,bZ=True)
    Parents=((Item=(Type=Bone,Name="None"),Weight=1))
    Weight=1
    WorldUp=(Target=(X=0,Y=0,Z=1),Kind=Direction,Space=(Type=None,Name="None"))
    bMaintainOffset=false
    UpAxis=@n11.Value
    WorldUp.Target=@n11.Value
    Child=RigVMVariableNode() {variable=secondBone}.Value
    AimAxis=RigVMVariableNode() {variable=primaryAxis}.Value
    WorldUp.Space=RigVMVariableNode() {variable=firstBone}.Value
    Parents.0.Item=RigVMVariableNode() {variable=ikEffector}.Value
    ExecutePin=@RigVMFunction_ControlFlowBranch.True
   )
   &RigUnit_ParentConstraint(
    AdvancedSettings=(InterpolationType=Average,RotationOrderForFilter=XZY)
    Parents=((Item=(Type=Bone,Name="None"),Weight=1))
    Weight=1
    bMaintainOffset=false
    Child=RigVMVariableNode() {variable=poleVector}.Value
    Parents.0.Item=RigVMVariableNode() {variable=secondBoneOrig}.Value
    ExecutePin=@n2.True
   )
   > &RigUnit_ModifyTransforms(
    ItemToModify=((Item=(Type=Bone,Name="None"),Transform=(Rotation=(X=0,Y=0,Z=0,W=1),Translation=(X=0,Y=0,Z=0),Scale3D=(X=1,Y=1,Z=1))))
    Weight=1
    WeightMinimum=0
    ItemToModify.0.Transform.Translation=RigVMVariableNode() {variable=offsetPv}.Value
    ItemToModify.0.Item=RigVMVariableNode() {variable=poleVector}.Value
   )
   n12 = RigVMVariableNode() {variable=secondaryAxis}
   > &RigUnit_AimConstraintLocalSpaceOffset(
    Filter=(bX=True,bY=True,bZ=True)
    Parents=((Item=(Type=Bone,Name="None"),Weight=1))
    Weight=1
    WorldUp=(Target=(X=0,Y=1,Z=0),Kind=Direction,Space=(Type=None,Name="None"))
    bMaintainOffset=false
    AimAxis=RigVMVariableNode() {variable=primaryAxis}.Value
    UpAxis=@n12.Value
    WorldUp.Target=@n12.Value
    WorldUp.Space=RigVMVariableNode() {variable=firstBoneOrig}.Value
    Parents.0.Item=RigVMVariableNode() {variable=poleVector}.Value
    Child=RigVMVariableNode() {variable=firstBone}.Value
   )
   > &RigUnit_PositionConstraintLocalSpaceOffset(
    Filter=(bX=True,bY=True,bZ=True)
    Parents=((Item=(Type=Bone,Name="None"),Weight=1))
    Weight=1
    bMaintainOffset=false
    Parents.0.Item=RigVMVariableNode() {variable=poleVector}.Value
    Child=RigVMVariableNode() {variable=secondBone}.Value
   )
   n13 = RigVMVariableNode() {variable=secondaryAxis}
   > &RigUnit_AimConstraintLocalSpaceOffset(
    Filter=(bX=True,bY=True,bZ=True)
    Parents=((Item=(Type=Bone,Name="None"),Weight=1))
    Weight=1
    WorldUp=(Target=(X=0,Y=1,Z=0),Kind=Direction,Space=(Type=None,Name="None"))
    bMaintainOffset=false
    Child=RigVMVariableNode() {variable=secondBone}.Value
    AimAxis=RigVMVariableNode() {variable=primaryAxis}.Value
    UpAxis=@n13.Value
    WorldUp.Target=@n13.Value
    WorldUp.Space=RigVMVariableNode() {variable=secondBoneOrig}.Value
    Parents.0.Item=RigVMVariableNode() {variable=ikEffector}.Value
   )
  ]}
 New Function_ContainedGraph {nodes [
   n1 = RigVMFunctionEntryNode()
   RigVMFunction_Sequence = &RigVMFunction_Sequence(ExecuteContext=@n1.ExecuteContext)
   n2 = &RigVMFunction_Sequence(ExecuteContext=@RigVMFunction_Sequence.A)
   TwoBoneIKWithSoftness v4 = RigVMFunctionReferenceNode(
    bendAxis=(X=0,Y=0,Z=1)
    effectorBone=(Type=Bone,Name="footRet_l")
    firstBone=(Type=Bone,Name="thighRet_l")
    firstBoneOrig=(Type=Bone,Name="thigh_l")
    ikEffector=(Type=Bone,Name="foot_ik_l")
    itemStartStretch=(Type=Null,Name="thighRet_l_null")
    lockPV=True
    offsetPv=(X=0,Y=0,Z=0)
    poleVector=(Type=Control,Name="leg_l_pv_ctrl")
    primaryAxis=(X=-1,Y=0,Z=0)
    secondBone=(Type=Bone,Name="calfRet_l")
    secondBoneConstraint=True
    secondBoneOrig=(Type=Bone,Name="calf_l")
    secondaryAxis=(X=0,Y=1,Z=0)
    softness=1
    ExecuteContext=@n2.A
   )
   > TwoBoneIKWithSoftness v4_1 = RigVMFunctionReferenceNode(
    bendAxis=(X=0,Y=0,Z=1)
    effectorBone=(Type=Bone,Name="footRet_r")
    firstBone=(Type=Bone,Name="thighRet_r")
    firstBoneOrig=(Type=Bone,Name="thigh_r")
    ikEffector=(Type=Bone,Name="foot_ik_r")
    itemStartStretch=(Type=Null,Name="thighRet_r_null")
    lockPV=True
    offsetPv=(X=0,Y=0,Z=0)
    poleVector=(Type=Control,Name="leg_r_pv_ctrl")
    primaryAxis=(X=1,Y=0,Z=0)
    secondBone=(Type=Bone,Name="calfRet_r")
    secondBoneConstraint=True
    secondBoneOrig=(Type=Bone,Name="calf_r")
    secondaryAxis=(X=0,Y=-1,Z=0)
    softness=1
   )
   n3 = &RigVMFunction_Sequence(ExecuteContext=@"TwoBoneIKWithSoftness v4_1".ExecuteContext)
   DeltaQuatExtraction_4 = RigVMFunctionReferenceNode(
    child=(Type=Bone,Name="thighRet_l")
    invert=true
    onlyTwist=True
    parent=(Type=Null,Name="thighRet_l_null")
    twistAxis=(X=1,Y=0,Z=0)
    ExecuteContext=@n3.A
   )
   ItemArray_2 = RigVMDispatchNode(Value=((Type=Bone,Name="thighRet_twist_01_l"),(Type=Bone,Name="thighRet_twist_02_l"),(Type=Bone,Name="thighRet_twist_03_l"))) {template="DISPATCH_RigVMDispatch_Constant(io Value)"}
   > SplineTwistTopBot_4 = RigVMFunctionReferenceNode(botRotQuat=(), enable=true, topRotQuat=@DeltaQuatExtraction_4.outputQuat, chainArray=@ItemArray_2.Value)
   > DeltaQuatExtraction = RigVMFunctionReferenceNode(child=(Type=Bone,Name="foot_l"), onlyTwist=True, parent=(Type=Null,Name="foot_l_null"), twistAxis=(X=1,Y=0,Z=0))
   ItemArray = RigVMDispatchNode(Value=((Type=Bone,Name="calfRet_twist_01_l"),(Type=Bone,Name="calfRet_twist_02_l"))) {template="DISPATCH_RigVMDispatch_Constant(io Value)"}
   > SplineTwistTopBot = RigVMFunctionReferenceNode(enable=true, topRotQuat=(), botRotQuat=@DeltaQuatExtraction.outputQuat, chainArray=@ItemArray.Value)
   > &RigUnit_AimConstraintLocalSpaceOffset(
    AimAxis=(X=0,Y=1,Z=0)
    Child=(Type=Bone,Name="metaTarsoRet_l")
    Filter=(bX=True,bY=true,bZ=true)
    Parents=((Item=(Type=Bone,Name="ball_l"),Weight=1))
    UpAxis=(X=0,Y=0,Z=-1)
    Weight=1
    WorldUp=(Target=(X=0,Y=0,Z=1),Kind=Location,Space=(Type=Null,Name="metaTarsoRet_l_upV_null"))
    bMaintainOffset=true
   )
   > &RigUnit_RotationConstraintLocalSpaceOffset(
    AdvancedSettings=(InterpolationType=Average,RotationOrderForFilter=XZY)
    Child=(Type=Bone,Name="toeRet_l")
    Filter=(bX=True,bY=True,bZ=True)
    Parents=((Item=(Type=Bone,Name="ball_l"),Weight=1))
    Weight=1
   )
   DeltaQuatExtraction_1 = RigVMFunctionReferenceNode(
    child=(Type=Bone,Name="thighRet_r")
    invert=true
    onlyTwist=True
    parent=(Type=Null,Name="thighRet_r_null")
    twistAxis=(X=1,Y=0,Z=0)
    ExecuteContext=@n3.B
   )
   ItemArray_6 = RigVMDispatchNode(Value=((Type=Bone,Name="thighRet_twist_01_r"),(Type=Bone,Name="thighRet_twist_02_r"),(Type=Bone,Name="thighRet_twist_03_r"))) {template="DISPATCH_RigVMDispatch_Constant(io Value)"}
   > SplineTwistTopBot_1 = RigVMFunctionReferenceNode(botRotQuat=(), enable=true, topRotQuat=@DeltaQuatExtraction_1.outputQuat, chainArray=@ItemArray_6.Value)
   > DeltaQuatExtraction_2 = RigVMFunctionReferenceNode(child=(Type=Bone,Name="foot_r"), onlyTwist=True, parent=(Type=Null,Name="foot_r_null"), twistAxis=(X=1,Y=0,Z=0))
   ItemArray_5 = RigVMDispatchNode(Value=((Type=Bone,Name="footRet_r"),(Type=Bone,Name="calf_twist_01_r"),(Type=Bone,Name="calf_twist_02_r"))) {template="DISPATCH_RigVMDispatch_Constant(io Value)"}
   > SplineTwistTopBot_2 = RigVMFunctionReferenceNode(enable=true, topRotQuat=(), botRotQuat=@DeltaQuatExtraction_2.outputQuat, chainArray=@ItemArray_5.Value)
   > &RigUnit_AimConstraintLocalSpaceOffset(
    AimAxis=(X=0,Y=-1,Z=0)
    Child=(Type=Bone,Name="metaTarsoRet_r")
    Filter=(bX=True,bY=true,bZ=true)
    Parents=((Item=(Type=Bone,Name="ball_r"),Weight=1))
    Weight=1
    WorldUp=(Target=(X=0,Y=0,Z=1),Kind=Location,Space=(Type=Null,Name="metaTarsoRet_r_upV_null"))
    bMaintainOffset=true
   )
   > &RigUnit_RotationConstraintLocalSpaceOffset(
    AdvancedSettings=(InterpolationType=Average,RotationOrderForFilter=XZY)
    Child=(Type=Bone,Name="toeRet_r")
    Filter=(bX=True,bY=True,bZ=True)
    Parents=((Item=(Type=Bone,Name="ball_r"),Weight=1))
    Weight=1
   )
   RigVMFunctionReturnNode(ExecuteContext=@RigVMFunction_Sequence.B)
   Knee Corrective = RigVMCollapseNode(Value=@n2.B)
   GroinGlute Corrective = RigVMCollapseNode(ExecutePin=@"Knee Corrective".Value)
   ItemArray_1 = RigVMDispatchNode(Value=((Type=Bone,Name="leg_bend_6_l"),(Type=Bone,Name="leg_bend_7_l"),(Type=Bone,Name="leg_bend_8_l"),(Type=Bone,Name="leg_bend_9_l"),(Type=Bone,Name="leg_bend_10_l"))) {template="DISPATCH_RigVMDispatch_Constant(io Value)"}
   DeformRig_GetDeltaRefTransform_FNC = RigVMFunctionReferenceNode(ExecuteContext=(), Item=(Type=Bone,Name="calf_l"))
   &RigVMFunction_MathFloatRemap(SourceMaximum=100, SourceMinimum=10, bClamp=true, Value=@DeformRig_GetDeltaRefTransform_FNC.OutTransform.Rotation.Yaw)
   DeformRig_GetDeltaRefTransform_FNC_1 = RigVMFunctionReferenceNode(ExecuteContext=(), Item=(Type=Bone,Name="calf_r"))
   &RigVMFunction_MathFloatRemap(SourceMaximum=100, SourceMinimum=10, bClamp=true, Value=@DeformRig_GetDeltaRefTransform_FNC_1.OutTransform.Rotation.Yaw)
   TwoBoneIKWithSoftness v4_2 = RigVMFunctionReferenceNode(
    ExecuteContext=()
    bendAxis=(X=0,Y=0,Z=1)
    effectorBone=(Type=Bone,Name="footRet_l")
    firstBone=(Type=Bone,Name="thighRet_l")
    firstBoneOrig=(Type=Bone,Name="thigh_l")
    ikEffector=(Type=Bone,Name="hand_ik_l")
    itemStartStretch=(Type=Null,Name="upperRetArm_l_null")
    lockPV=false
    offsetPv=(X=0,Y=0,Z=30)
    poleVector=(Type=Bone,Name="calf_l")
    primaryAxis=(X=1,Y=0,Z=0)
    secondBone=(Type=Bone,Name="lowerRetArm_l")
    secondBoneConstraint=true
    secondBoneOrig=(Type=Bone,Name="calf_l")
    secondaryAxis=(X=0,Y=-1,Z=0)
    softness=2
   )
  ]}
 New Function_ContainedGraph {nodes [
   n1 = RigVMFunctionEntryNode()
   RigVMFunction_MathBoolToInteger = &RigVMFunction_MathBoolToInteger(Value=@n1.onlyTwist)
   RigVMFunction_MathQuaternionSwingTwist = &RigVMFunction_MathQuaternionSwingTwist(
    twistAxis=@n1.twistAxis
    Input=&RigUnit_GetRelativeTransformForItem(bChildInitial=False, bParentInitial=False, child=@n1.child, parent=@n1.parent).RelativeTransform.Rotation
   )
   n2 = RigVMDispatchNode(
    Values=((),())
    Index=@RigVMFunction_MathBoolToInteger.Result
    Values.0=@RigVMFunction_MathQuaternionSwingTwist.Swing
    Values.1=@RigVMFunction_MathQuaternionSwingTwist.Twist
   ) {template="DISPATCH_RigVMDispatch_SelectInt32(in Index,in Values,out Result)"}
   Inverse = &RigVMFunction_MathQuaternionInverse(Value=@n2.Result) {template="Inverse::Execute(in Value,out Result)"}
   > RigVMFunctionReturnNode(outputQuat=RigVMDispatchNode(True=@Inverse.Result, Condition=@n1.invert, False=@n2.Result) {template="DISPATCH_RigVMDispatch_If(in Condition,in True,in False,out Result)"}.Result)
  ]}
 New Function_ContainedGraph {nodes [
   n1 = RigVMFunctionEntryNode()
   RigVMFunction_ControlFlowBranch = &RigVMFunction_ControlFlowBranch(ExecuteContext=@n1.ExecuteContext, Condition=@n1.enable)
   RigVMFunctionReturnNode(ExecuteContext=@RigVMFunction_ControlFlowBranch.Completed)
   For_Each = RigVMDispatchNode(Array=@n1.chainArray, ExecuteContext=@RigVMFunction_ControlFlowBranch.True) {template="DISPATCH_RigVMDispatch_ArrayIterator(in Array,out Element,out Index,out Count,out Ratio)"}
   Interpolate = &RigVMFunction_MathQuaternionSlerp(T=@For_Each.Ratio, B=@n1.botRotQuat, A=@n1.topRotQuat) {template="Interpolate::Execute(in A,in B,in T,out Result)"}
   > &RigUnit_OffsetTransformForItem(
    OffsetTransform=(Rotation=(X=0,Y=0,Z=0,W=1),Translation=(X=0,Y=0,Z=0),Scale3D=(X=1,Y=1,Z=1))
    Weight=1
    bPropagateToChildren=false
    OffsetTransform.Rotation=@Interpolate.Result
    Item=@For_Each.Element
   )
  ]}
 RigVMFunctionLibrary {nodes [
   Arm IK = RigVMCollapseNode(Argument=())
   DeformRig_GetDeltaRefTransform_FNC = RigVMCollapseNode(ExecuteContext=(), Item=(Type=Bone))
   DeformRig_GetMirrorTransform_FNC = RigVMCollapseNode(ExecuteContext=(), transform=(Rotation=(X=0,Y=0,Z=0,W=1),Translation=(X=0,Y=0,Z=0),Scale3D=(X=1,Y=1,Z=1)))
   DeltaQuatExtraction = RigVMCollapseNode(ExecuteContext=(), child=(Type=Bone,Name="None"), parent=(Type=Bone,Name="None"), twistAxis=(X=1,Y=0,Z=0))
   DeltaQuatExtraction_1 = RigVMCollapseNode(ExecuteContext=(), child=(Type=Bone,Name="None"), parent=(Type=Bone,Name="None"), twistAxis=(X=1,Y=0,Z=0))
   Leg IK = RigVMCollapseNode(ExecuteContext=())
   SplineTwistTopBot = RigVMCollapseNode(
    ExecuteContext=()
    botRotQuat=()
    chainArray="((Type=Bone,Name=\"arm_bend_1_l\"),(Type=Bone,Name=\"arm_bend_2_l\"),(Type=Bone,Name=\"arm_bend_3_l\"),(Type=Bone,Name=\"arm_bend_4_l\"),(Type=Bone,Name=\"arm_bend_5_l\"),(Type=Bone,Name=\"arm_bend_6_l\"),(Type=Bone,Name=\"arm_bend_7_l\"),(Type=Bone,Name=\"arm_bend_8_l\") ...(truncated)"
    enable=true
    topRotQuat=()
   )
   SplineTwistTopBot_1 = RigVMCollapseNode(
    ExecuteContext=()
    botRotQuat=()
    chainArray="((Type=Bone,Name=\"arm_bend_1_l\"),(Type=Bone,Name=\"arm_bend_2_l\"),(Type=Bone,Name=\"arm_bend_3_l\"),(Type=Bone,Name=\"arm_bend_4_l\"),(Type=Bone,Name=\"arm_bend_5_l\"),(Type=Bone,Name=\"arm_bend_6_l\"),(Type=Bone,Name=\"arm_bend_7_l\"),(Type=Bone,Name=\"arm_bend_8_l\") ...(truncated)"
    enable=true
    topRotQuat=()
   )
   TwoBoneIKWithSoftness v4 = RigVMCollapseNode(
    ExecuteContext=()
    bendAxis=(X=0,Y=0,Z=1)
    effectorBone=(Type=None,Name="None")
    firstBone=(Type=None,Name="None")
    firstBoneOrig=(Type=None,Name="None")
    ikEffector=(Type=None,Name="None")
    itemStartStretch=(Type=None,Name="None")
    offsetPv=(X=0,Y=0,Z=0)
    poleVector=(Type=None,Name="None")
    primaryAxis=(X=1,Y=0,Z=0)
    secondBone=(Type=None,Name="None")
    secondBoneOrig=(Type=Bone,Name="None")
    secondaryAxis=(X=0,Y=1,Z=0)
    softness=1
   )
   TwoBoneIKWithSoftness v4_1 = RigVMCollapseNode(
    ExecuteContext=()
    bendAxis=(X=0,Y=0,Z=1)
    effectorBone=(Type=None,Name="None")
    firstBone=(Type=None,Name="None")
    firstBoneOrig=(Type=None,Name="None")
    ikEffector=(Type=None,Name="None")
    itemStartStretch=(Type=None,Name="None")
    offsetPv=(X=0,Y=0,Z=0)
    poleVector=(Type=None,Name="None")
    primaryAxis=(X=1,Y=0,Z=0)
    secondBone=(Type=None,Name="None")
    secondBoneOrig=(Type=Bone,Name="None")
    secondaryAxis=(X=0,Y=1,Z=0)
    softness=1
   )
  ]}
 RigVMFunction_Sequence_1_ContainedGraph {nodes [
   n1 = RigVMFunctionEntryNode()
   n2 = &RigVMFunction_Sequence(ExecuteContext=@n1.ExecuteContext)
   n3 = &RigVMFunction_Sequence(ExecuteContext=@n2.B)
   n4 = &RigVMFunction_Sequence(ExecuteContext=@n3.B)
   RigVMFunctionReturnNode(B=@n3.A, A=@n2.A, C=@n4.A, D=@n4.B)
  ]}
 RigVMFunction_Sequence_1_ContainedGraph {nodes [
   n1 = RigVMFunctionEntryNode()
   RigVMFunction_Sequence = &RigVMFunction_Sequence(ExecuteContext=@n1.ExecuteContext)
   n2 = &RigVMFunction_Sequence(ExecuteContext=@RigVMFunction_Sequence.B)
   RigVMFunctionReturnNode(A=@RigVMFunction_Sequence.A, B=@n2.A, C=@n2.B)
  ]}
 RigVMFunction_Sequence_4_ContainedGraph {nodes [
   n1 = RigVMFunctionEntryNode()
   RigVMFunction_Sequence = &RigVMFunction_Sequence(ExecuteContext=@n1.ExecuteContext)
   n2 = &RigVMFunction_Sequence(ExecuteContext=@RigVMFunction_Sequence.B)
   RigVMFunctionReturnNode(A=@RigVMFunction_Sequence.A, B=@n2.A, C=@n2.B)
  ]}
 RigVMFunction_Sequence_6_ContainedGraph {nodes [
   n1 = RigVMFunctionEntryNode()
   n2 = &RigVMFunction_Sequence(ExecuteContext=@n1.ExecuteContext)
   n3 = &RigVMFunction_Sequence(ExecuteContext=@n2.B)
   RigVMFunctionReturnNode(B=@n3.A, C=@n3.B, A=@n2.A)
  ]}
 RigVMFunction_Sequence_6_ContainedGraph {nodes [
   n1 = RigVMFunctionEntryNode()
   n2 = &RigVMFunction_Sequence(ExecuteContext=@n1.ExecuteContext)
   n3 = &RigVMFunction_Sequence(ExecuteContext=@n2.B)
   RigVMFunctionReturnNode(B=@n3.A, C=@n3.B, A=@n2.A)
  ]}
 RigVMModel {nodes [
   n1 = &RigVMFunction_Sequence(ExecuteContext=@RigUnit_BeginExecution.ExecutePin)
   Arm IK = RigVMFunctionReferenceNode(Argument=@n1.B)
   Leg IK = RigVMFunctionReferenceNode(ExecuteContext=@"Arm IK".Argument)
   RigUnit_BeginExecution = &RigUnit_BeginExecution()
   FN_DeformRig_FNC = RigVMFunctionReferenceNode(
    DeformRig_ Curve="(deformRig_l_upperarm_up_alpha_72_#1=False,deformRig_r_upperarm_up_alpha_73_#2=False,deformRig_l_upperarm_fr_alpha_74_#3=False,deformRig_r_upperarm_fr_alpha_75_CB51B8 ...(truncated)"
    DeformRig_Bool="(upperarm_1_#4=True,elbow_3_#5=True,wrist_7_#6=True,groin_8_#7=True,glute_17_#8=True,knee_9_40C408324CA261 ...(truncated)"
    DeformRig_Debug=(upperarm_debug_18_#4=False,wrist_debug_20_#6=False,lat_debug_26_#9=False,groinGlute_debug_27_#7=False)
    DeformRig_Value="(upperarm_up_18_#10=(Rotation=(X=0,Y=0,Z=0,W=1),Translation=(X=0,Y=0,Z=8),Scale3D=(X=0.7,Y=1,Z=1)),upperarm_fwd_fr_19_#11=(Rotation=(X=0,Y=0,Z=0,W=1),Translation=(X=0,Y=10,Z=0),Scale3D=(X=0.5,Y=1,Z= ...(truncated)"
    ExecuteContext=()
   )
   /Script/FortniteGame.RigUnit_FortDeformRig(
    Bool=(bUpperArm=False,bElbow=False,bWrist=true,bGroin=true,bGlute=true,bKnee=False,bPec=False,bBicep=False,bLat=true)
    Curve="(bLUpperArmUpAlpha=False,bRUpperArmUpAlpha=False,bLUpperArmFrAlpha=False,bRUpperArmFrAlpha=False,bLUpperArmBkAlpha=False,bRUpperArmBkAlpha=False,bLWristUpAlpha=False,bRWristUpAlpha=False,bLWristDnAlpha=False,bRWristDnAlpha=False,bLKneeAlpha=False,bRKneeAlp ...(truncated)"
    Debug=(bUpperArmDebug=False,bWristDebug=False,bLatDebug=False,bGroinGluteDebug=False)
    Value="(UpperArmUp=(Rotation=(X=0,Y=0,Z=0,W=1),Translation=(X=0,Y=0,Z=0),Scale3D=(X=1,Y=1,Z=1)),UpperArmFwdFr=(Rotation=(X=0,Y=0,Z=0,W=1),Translation=(X=0,Y=0,Z=0),Scale3D=(X=1,Y=1,Z=1)),UpperArmFwdBk=(Rotation=(X=0,Y=0,Z=0,W=1),Translation=(X=0,Y=0,Z=0),Scale3D= ...(truncated)"
    ExecutePin=@n1.A
   ) {is=struct RigVMUnitNode}
  ]}
 RigVMModel Construction Graph {nodes [
   n1 = &RigUnit_PrepareForExecution()
   RigVMFunction_Sequence = &RigVMFunction_Sequence(ExecuteContext=@n1.ExecutePin)
   ItemArray = RigVMDispatchNode(Value="((Type=Bone,Name=\"upperarm_l\"),(Type=Bone,Name=\"upperarm_r\"),(Type=Bone,Name=\"upperRetArm_l\"),(Type=Bone,Name=\"upperRetArm_r\"),(Type=Bone,Name=\"hand_l\"),(Type=Bone,Name=\"hand_r\"),(Type=Bone,Name=\"lowerRetArm_l\"),(Type=Bone,Name=\"lowerRetArm_r\"),(Type=Bone, ...(truncated)") {template="DISPATCH_RigVMDispatch_Constant(io Value)"}
   DISPATCH_RigVMDispatch_ArrayIterator = RigVMDispatchNode(Array=@ItemArray.Value, ExecuteContext=@RigVMFunction_Sequence.A) {template="DISPATCH_RigVMDispatch_ArrayIterator(in Array,out Element,out Index,out Count,out Ratio)"}
   Concat = /Script/RigVM.RigVMFunction_NameConcat(B=_null, A=@DISPATCH_RigVMDispatch_ArrayIterator.Element.Name) {is=struct RigVMUnitNode, template="Concat::Execute(in A,in B,out Result)"}
   > &RigUnit_HierarchyAddNull(
    Space=LocalSpace
    Name=@Concat.Result
    Transform=&RigUnit_GetTransform(Item=(Type=Bone,Name="clavicle_l"), Space=LocalSpace, bInitial=true, Item.Name=@DISPATCH_RigVMDispatch_ArrayIterator.Element.Name).Transform
    Parent=/Script/ControlRig.RigUnit_HierarchyGetParent(bDefaultParent=True, Child=@DISPATCH_RigVMDispatch_ArrayIterator.Element) {is=struct RigVMUnitNode}.Parent
   )
   n2 = &RigVMFunction_Sequence(ExecuteContext=@DISPATCH_RigVMDispatch_ArrayIterator.Completed)
   SpawnControl = /Script/ControlRig.RigUnit_HierarchyAddControlTransform(
    InitialValue=(Rotation=(X=0,Y=0,Z=0,W=1),Translation=(X=0,Y=0,Z=0),Scale3D=(X=1,Y=1,Z=1))
    Name=arm_l_pv_ctrl
    OffsetSpace=LocalSpace
    OffsetTransform=(Rotation=(X=0,Y=0,Z=0,W=1),Translation=(X=0,Y=0,Z=0),Scale3D=(X=1,Y=1,Z=1))
    Parent=(Type=Bone,Name="None")
    Settings="(InitialSpace=LocalSpace,bUsePreferredRotationOrder=False,PreferredRotationOrder=YZX,Limits=(LimitTranslationX=(bMinimum=False,bMaximum=False),LimitTranslationY=(bMinimum=False,bMaximum=False),LimitTranslationZ=(bMinimum=False,bMaximum=False),LimitPitch=(b ...(truncated)"
    ExecutePin=@n2.A
   ) {
    is=struct RigVMUnitNode
    template="SpawnControl::Execute(in InitialValue,in Settings,in OffsetTransform,in OffsetSpace,in Parent,in Name,out Item)"
   }
   > &RigUnit_HierarchyAddControlTransform(
    InitialValue=(Rotation=(X=0,Y=0,Z=0,W=1),Translation=(X=0,Y=0,Z=0),Scale3D=(X=1,Y=1,Z=1))
    Name=arm_r_pv_ctrl
    OffsetSpace=LocalSpace
    OffsetTransform=(Rotation=(X=0,Y=0,Z=0,W=1),Translation=(X=0,Y=0,Z=0),Scale3D=(X=1,Y=1,Z=1))
    Parent=(Type=Bone,Name="None")
    Settings="(InitialSpace=LocalSpace,bUsePreferredRotationOrder=False,PreferredRotationOrder=YZX,Limits=(LimitTranslationX=(bMinimum=False,bMaximum=False),LimitTranslationY=(bMinimum=False,bMaximum=False),LimitTranslationZ=(bMinimum=False,bMaximum=False),LimitPitch=(b ...(truncated)"
   ) {template="SpawnControl::Execute(in InitialValue,in Settings,in OffsetTransform,in OffsetSpace,in Parent,in Name,out Item)"}
   > ComputePoleVector = RigVMFunctionReferenceNode(
    Bone A=(Type=Bone,Name="upperarm_l")
    Bone B=(Type=Bone,Name="lowerarm_l")
    Bone C=(Type=Bone,Name="hand_l")
    Draw Transform=true
    Initial=true
    OffsetFactor=10
    Origin Scale=(X=0,Y=0,Z=0)
   )
   > &RigUnit_SetTransform(
    Item=(Type=Control,Name="arm_l_pv_ctrl")
    Value=(Rotation=(X=0,Y=0,Z=0,W=1),Translation=(X=0,Y=0,Z=0),Scale3D=(X=1,Y=1,Z=1))
    Weight=1
    bInitial=true
    Value.Translation=@ComputePoleVector.Transform.Translation
   )
   > ComputePoleVector_1 = RigVMFunctionReferenceNode(
    Bone A=(Type=Bone,Name="upperarm_r")
    Bone B=(Type=Bone,Name="lowerarm_r")
    Bone C=(Type=Bone,Name="hand_r")
    Draw Transform=true
    Initial=true
    OffsetFactor=10
    Origin Scale=(X=0,Y=0,Z=0)
   )
   > &RigUnit_SetTransform(
    Item=(Type=Control,Name="arm_r_pv_ctrl")
    Value=(Rotation=(X=0,Y=0,Z=0,W=1),Translation=(X=0,Y=0,Z=0),Scale3D=(X=1,Y=1,Z=1))
    Weight=1
    bInitial=true
    Value.Translation=@ComputePoleVector_1.Transform.Translation
   )
   > /Script/ControlRig.RigUnit_HierarchyAddBone(
    Name=hand_ik_l
    Parent=(Type=Bone,Name="hand_l")
    Space=LocalSpace
    Transform=(Rotation=(X=0,Y=0,Z=0,W=1),Translation=(X=0,Y=0,Z=0),Scale3D=(X=1,Y=1,Z=1))
   ) {is=struct RigVMUnitNode}
   > &RigUnit_HierarchyAddBone(
    Name=hand_ik_r
    Parent=(Type=Bone,Name="hand_r")
    Space=LocalSpace
    Transform=(Rotation=(X=0,Y=0,Z=0,W=1),Translation=(X=0,Y=0,Z=0),Scale3D=(X=1,Y=1,Z=1))
   )
   > ComputePoleVector_2 = RigVMFunctionReferenceNode(
    Bone A=(Type=Bone,Name="thighRet_l")
    Bone B=(Type=Bone,Name="calfRet_l")
    Bone C=(Type=Bone,Name="footRet_l")
    Draw Transform=true
    Initial=false
    OffsetFactor=100
    Origin Scale=(X=0,Y=0,Z=0)
   )
   > /Script/ControlRig.RigUnit_HierarchyAddControlFloat(
    InitialValue=0
    Name=leg_l_pv_ctrl
    OffsetSpace=LocalSpace
    OffsetTransform=(Rotation=(X=0,Y=0,Z=0,W=1),Translation=(X=0,Y=0,Z=0),Scale3D=(X=1,Y=1,Z=1))
    Parent=(Type=Bone,Name="None")
    Settings="(PrimaryAxis=X,bIsScale=False,Limits=(Limit=(bMinimum=false,bMaximum=false),MinValue=0,MaxValue=100,bDrawLimits=false),Shape=(bVisible=True,Name=\"Sphere_Thick\",Color=(R=0,G=0,B=1,A=1),Transform=(Rotation=(X=0,Y=0,Z=0,W=1),Translation=(X=0,Y=0,Z=0),Scale3D= ...(truncated)"
    OffsetTransform.Translation=@ComputePoleVector_2.Transform.Translation
   ) {
    is=struct RigVMUnitNode
    template="SpawnControl::Execute(in InitialValue,in Settings,in OffsetTransform,in OffsetSpace,in Parent,in Name,out Item)"
   }
   > ComputePoleVector_3 = RigVMFunctionReferenceNode(
    Bone A=(Type=Bone,Name="thighRet_r")
    Bone B=(Type=Bone,Name="calfRet_r")
    Bone C=(Type=Bone,Name="footRet_r")
    Draw Transform=true
    Initial=false
    OffsetFactor=100
    Origin Scale=(X=0,Y=0,Z=0)
   )
   > &RigUnit_HierarchyAddControlFloat(
    InitialValue=0
    Name=leg_r_pv_ctrl
    OffsetSpace=LocalSpace
    OffsetTransform=(Rotation=(X=0,Y=0,Z=0,W=1),Translation=(X=0,Y=0,Z=0),Scale3D=(X=1,Y=1,Z=1))
    Parent=(Type=Bone,Name="None")
    Settings="(PrimaryAxis=X,bIsScale=False,Limits=(Limit=(bMinimum=false,bMaximum=false),MinValue=0,MaxValue=100,bDrawLimits=false),Shape=(bVisible=True,Name=\"Sphere_Thick\",Color=(R=1,G=0,B=0,A=1),Transform=(Rotation=(X=0,Y=0,Z=0,W=1),Translation=(X=0,Y=0,Z=0),Scale3D= ...(truncated)"
    OffsetTransform.Translation=@ComputePoleVector_3.Transform.Translation
   ) {template="SpawnControl::Execute(in InitialValue,in Settings,in OffsetTransform,in OffsetSpace,in Parent,in Name,out Item)"}
   &RigUnit_HierarchyAddBone(
    Name=foot_ik_l
    Parent=(Type=Bone,Name="foot_l")
    Space=LocalSpace
    Transform=&RigUnit_GetRelativeTransformForItem(Child=(Type=Bone,Name="footRet_l"), Parent=(Type=Bone,Name="foot_l"), bChildInitial=true, bParentInitial=true).RelativeTransform
    ExecutePin=@n2.B
   )
   > &RigUnit_HierarchyAddBone(
    Name=foot_ik_r
    Parent=(Type=Bone,Name="foot_r")
    Space=LocalSpace
    Transform=&RigUnit_GetRelativeTransformForItem(Child=(Type=Bone,Name="footRet_r"), Parent=(Type=Bone,Name="foot_r"), bChildInitial=true, bParentInitial=true).RelativeTransform
   )
   Multiply = &RigVMFunction_MathTransformMul(
    A=(Rotation=(X=0,Y=0,Z=0,W=1),Translation=(X=0,Y=0,Z=-10),Scale3D=(X=1,Y=1,Z=1))
    B=&RigUnit_GetRelativeTransformForItem(Child=(Type=Bone,Name="metaTarsoRet_l"), Parent=(Type=Bone,Name="foot_l"), bChildInitial=true, bParentInitial=true).RelativeTransform
   ) {template="Multiply::Execute(in A,in B,out Result)"}
   > &RigUnit_HierarchyAddNull(Name=metaTarsoRet_l_upV_null, Parent=(Type=Bone,Name="foot_l"), Space=LocalSpace, Transform=@Multiply.Result)
   > &RigUnit_HierarchyAddNull(
    Name=metaTarsoRet_r_upV_null
    Parent=(Type=Bone,Name="foot_r")
    Space=LocalSpace
    Transform=&RigVMFunction_MathTransformMul(
     A=(Rotation=(X=0,Y=0,Z=0,W=1),Translation=(X=0,Y=0,Z=10),Scale3D=(X=1,Y=1,Z=1))
     B=&RigUnit_GetRelativeTransformForItem(Child=(Type=Bone,Name="metaTarsoRet_r"), Parent=(Type=Bone,Name="foot_r"), bChildInitial=true, bParentInitial=true).RelativeTransform
    ) {template="Multiply::Execute(in A,in B,out Result)"}.Result
   )
   addFKCtrl = RigVMFunctionReferenceNode(
    Custom_FKCtrl=()
    Options_FKCtrl="(pelvis_1_#12=true,spine_3_#13=true,clavicle_5_#14=true,arm_7_#15=true,leg_9_#16=true,head_11_46DC15EF465CD2979 ...(truncated)"
    Shape_Settings="(pelvis_2_#17=(bVisible=True,Name=\"Hexagon_Thin\",Color=(R=0.0117,G=0.8281,B=1,A=1),Transform=(Rotation=(X=0,Y=-0.7071,Z=0,W=0.7071),Translation=(X=2,Y=0,Z=0),Scale3D=(X=5,Y=5,Z=7))),spine_4_#18=(bVi ...(truncated)"
    ExecuteContext=@RigVMFunction_Sequence.B
   )
   RigVMRerouteNode(Value=())
  ]}
]
hierarchy [
 root {loc=<0 0 0>, rot=<0 0 0 1>}
 attach {parent=root, loc=<0 0 0>, rot=<0 0 0 1>}
 pelvis {parent=root, loc=<0 -0.7321 92.4218>, rot=<0.0127 0.707 -0.0127 -0.707>}
 spine_01 {parent=pelvis, loc=<5.6002 -0.5128 0>, rot=<0 0 0.034 -0.9994>}
 spine_02 {parent=spine_01, loc=<8.0753 0.0139 0>, rot=<0 0 0.0238 -0.9997>}
 spine_03 {parent=spine_02, loc=<7.0393 0 0>, rot=<0 0 0.0149 -0.9999>}
 spine_04 {parent=spine_03, loc=<8.1024 0 0>, rot=<0 0 -0.0132 -0.9999>}
 spine_05 {parent=spine_04, loc=<10.0146 0 0>, rot=<0 0 -0.0015 -1>}
 clavicle_l {parent=spine_05, loc=<2.5856 -0.2866 -4.3777>, rot=<0.0465 -0.7289 -0.009 -0.6829>}
 upperarm_l {parent=clavicle_l, loc=<18.2264 0.1342 -0.001>, rot=<0.0253 -0.2693 0.0013 -0.9627>}
 lowerarm_l {parent=upperarm_l, loc=<29.0071 0.1686 -1.239>, rot=<0.0391 -0.0099 -0.2453 -0.9686>}
 hand_l {parent=lowerarm_l, loc=<25.8607 0.222 1.0274>, rot=<-0.7193 -0.0988 0.1193 -0.6773>}
 index_metacarpal_l {parent=hand_l, loc=<1.6535 0.0183 -2.96>, rot=<0.0158 -0.0124 -0.0173 -0.9996>}
 index_01_l {parent=index_metacarpal_l, loc=<8.3008 0.0357 -0.0524>, rot=<-0.0044 0.0485 0.0924 -0.9945>}
 index_02_l {parent=index_01_l, loc=<3.3587 -0.3004 0.0452>, rot=<-0.002 0.0027 0.3225 -0.9466>}
 index_03_l {parent=index_02_l, loc=<2.5578 -0.123 -0.0252>, rot=<-0.002 0.0067 0.2558 -0.9667>}
 middle_metacarpal_l {parent=hand_l, loc=<2.1856 -0.1725 -0.5984>, rot=<0.006 0.0168 -0.0329 -0.9993>}
 middle_01_l {parent=middle_metacarpal_l, loc=<7.3275 -0.2447 0.0129>, rot=<-0.0208 0.0042 0.0858 -0.9961>}
 middle_02_l {parent=middle_01_l, loc=<4.0123 0.1401 0.1175>, rot=<-0.0218 0.0258 0.3512 -0.9357>}
 middle_03_l {parent=middle_02_l, loc=<2.6077 0.0534 0.1373>, rot=<0.0082 0.0188 0.1708 -0.9851>}
 pinky_metacarpal_l {parent=hand_l, loc=<1.7616 -0.9753 3.3814>, rot=<-0.0397 0.0814 -0.069 -0.9935>}
 pinky_01_l {parent=pinky_metacarpal_l, loc=<7.1709 -0.4831 -0.1094>, rot=<-0.0654 -0.0035 0.2242 -0.9723>}
 pinky_02_l {parent=pinky_01_l, loc=<2.8274 0.0893 -0.1043>, rot=<0.0037 0.0011 0.2854 -0.9584>}
 pinky_03_l {parent=pinky_02_l, loc=<2.2046 0.0307 -0.0078>, rot=<-0.0002 0.0098 0.1445 -0.9895>}
 ring_metacarpal_l {parent=hand_l, loc=<2.2074 -0.3625 1.3638>, rot=<-0.0284 0.0722 -0.0458 -0.9959>}
 ring_01_l {parent=ring_metacarpal_l, loc=<7.0295 -0.6197 -0.1446>, rot=<-0.0489 -0.0252 0.1892 -0.9804>}
 ring_02_l {parent=ring_01_l, loc=<3.6174 0.4647 -0.1027>, rot=<0 0 0.2806 -0.9598>}
 ring_03_l {parent=ring_02_l, loc=<2.4152 0.114 0.0199>, rot=<0 0 0.1189 -0.9929>}
 deform_ring_01_lf {parent=ring_metacarpal_l, loc=<7.6155 -0.2401 -1.3885>, rot=<0.6877 -0.1209 0.0907 -0.7101>}
 deform_ring_02_lf {parent=deform_ring_01_lf, loc=<3.4588 0 0>, rot=<0 -0.3356 0 -0.942>}
 deform_ring_03_lf {parent=deform_ring_02_lf, loc=<2.2136 0 0>, rot=<0 0 0 1>}
 thumb_01_l {parent=hand_l, loc=<1.3118 -0.3191 -2.459>, rot=<0.6546 -0.4042 0.0664 -0.6354>}
 thumb_02_l {parent=thumb_01_l, loc=<5.9336 -0.0448 0.0401>, rot=<-0.0037 0.0213 0.298 -0.9543>}
 thumb_03_l {parent=thumb_02_l, loc=<2.2223 0.3682 -0.0907>, rot=<0.0118 -0.013 0.0066 -0.9998>}
 weapon_l {parent=hand_l, loc=<-0.1974 -3.4111 0.3811>, rot=<0 0 0 1>}
 deform_wrist_dn_base_l {parent=hand_l, loc=<0 -3.4173 0>, rot=<0 0 0 -1>}
 deform_wrist_dn_l {parent=deform_wrist_dn_base_l, loc=<0 0 0>, rot=<0 0 0 -1>}
 deform_wrist_up_base_l {parent=hand_l, loc=<0 2.7805 0>, rot=<0 0 0 -1>}
 deform_wrist_up_l {parent=deform_wrist_up_base_l, loc=<0 0 0>, rot=<0 0 0 -1>}
 lowerarm_twist_01_l {parent=lowerarm_l, loc=<17.8549 0 0>, rot=<0 0 0 1>}
 lowerarm_twist_02_l {parent=lowerarm_l, loc=<9.6199 0 0>, rot=<0 0 0 1>}
 deform_elbow_in_base_l {parent=lowerarm_l, loc=<0.2112 5.645 0.8274>, rot=<0 0 0 1>}
 deform_elbow_in_l {parent=deform_elbow_in_base_l, loc=<0 0 0>, rot=<0 0 0 1>}
 deform_elbow_out_base_l {parent=lowerarm_l, loc=<0 -7.5129 0>, rot=<0 0 0 1>}
 deform_elbow_out_l {parent=deform_elbow_out_base_l, loc=<0 0 0>, rot=<0 0 0 1>}
 upperarm_twist_01_l {parent=upperarm_l, loc=<7.9796 0.0843 -0.6195>, rot=<0 0 0 1>}
 deform_upperarm_bk_base_l {parent=upperarm_twist_01_l, loc=<-6.8983 -13.3826 -1.3258>, rot=<0 0 0 1>}
 deform_upperarm_bk_l {parent=deform_upperarm_bk_base_l, loc=<0 0 0>, rot=<0 0 0 1>}
 deform_upperarm_fr_base_l {parent=upperarm_twist_01_l, loc=<-8.3487 10.7117 -2.1455>, rot=<0 0 0 1>}
 deform_upperarm_fr_l {parent=deform_upperarm_fr_base_l, loc=<0 0 0>, rot=<0 0 0 1>}
 deform_upperarm_up_base_l {parent=upperarm_twist_01_l, loc=<-11.2012 -1.2234 8.7784>, rot=<0 0 0 1>}
 deform_upperarm_up_l {parent=deform_upperarm_up_base_l, loc=<0 0 0>, rot=<0 0 0 1>}
 upperarm_twist_02_l {parent=upperarm_l, loc=<19.5059 0.0843 -0.6195>, rot=<0 0 0 1>}
 deform_bicep_base_l {parent=upperarm_twist_02_l, loc=<0 8.0494 0>, rot=<0 0 0 1>}
 deform_bicep_l {parent=deform_bicep_base_l, loc=<0 0 0>, rot=<0 0 0 1>}
 upperRetArm_l {parent=clavicle_l, loc=<23.6836 -0.248 -2.4578>, rot=<0.0252 -0.2897 0.0018 -0.9568>}
 lowerRetArm_l {parent=upperRetArm_l, loc=<23.142 0 0>, rot=<0 0 -0.2611 -0.9653>}
 handRet_l {parent=lowerRetArm_l, loc=<25.8607 0 0>, rot=<0 0 0 1>}
 lowerRetArm_twist_01_l {parent=lowerRetArm_l, loc=<6.4652 0 0>, rot=<0 0 0 1>}
 lowerRetArm_twist_02_l {parent=lowerRetArm_l, loc=<12.9304 0 0>, rot=<0 0 0 1>}
 lowerRetArm_twist_03_l {parent=lowerRetArm_l, loc=<19.3955 0 0>, rot=<0 0 0 1>}
 lowerRetArm_twist_04_l {parent=lowerRetArm_l, loc=<22.8801 0 0>, rot=<0 0 0 1>}
 upperRetArm_twist_01_l {parent=upperRetArm_l, loc=<1 0 0>, rot=<0 0 0 1>}
 upperRetArm_twist_02_l {parent=upperRetArm_l, loc=<7.714 0 0>, rot=<0 0 0 1>}
 upperRetArm_twist_03_l {parent=upperRetArm_l, loc=<15.428 0 0>, rot=<0 0 0 1>}
 clavicle_r {parent=spine_05, loc=<2.5856 -0.2865 4.3777>, rot=<0.7289 0.0465 -0.6829 0.009>}
 upperarm_r {parent=clavicle_r, loc=<-18.2264 -0.1342 0.001>, rot=<0.0253 -0.2693 0.0013 -0.9627>}
 lowerarm_r {parent=upperarm_r, loc=<-29.0072 -0.1686 1.239>, rot=<0.0391 -0.0099 -0.2453 -0.9686>}
 hand_r {parent=lowerarm_r, loc=<-25.8608 -0.222 -1.0274>, rot=<-0.7193 -0.0988 0.1193 -0.6773>}
 index_metacarpal_r {parent=hand_r, loc=<-1.6534 -0.0181 2.9601>, rot=<0.0158 -0.0124 -0.0173 -0.9996>}
 index_01_r {parent=index_metacarpal_r, loc=<-8.3008 -0.0357 0.0524>, rot=<-0.0044 0.0485 0.0924 -0.9945>}
 index_02_r {parent=index_01_r, loc=<-3.3589 0.3003 -0.0453>, rot=<-0.002 0.0027 0.3225 -0.9466>}
 index_03_r {parent=index_02_r, loc=<-2.5578 0.123 0.0252>, rot=<-0.002 0.0067 0.2558 -0.9667>}
 middle_metacarpal_r {parent=hand_r, loc=<-2.1856 0.1726 0.5984>, rot=<0.006 0.0168 -0.0329 -0.9993>}
 middle_01_r {parent=middle_metacarpal_r, loc=<-7.3273 0.2447 -0.0129>, rot=<-0.0208 0.0042 0.0858 -0.9961>}
 middle_02_r {parent=middle_01_r, loc=<-4.0124 -0.14 -0.1175>, rot=<-0.0218 0.0258 0.3512 -0.9357>}
 middle_03_r {parent=middle_02_r, loc=<-2.6077 -0.0533 -0.1373>, rot=<0.0082 0.0188 0.1708 -0.9851>}
 pinky_metacarpal_r {parent=hand_r, loc=<-1.7617 0.9753 -3.3814>, rot=<-0.0397 0.0814 -0.069 -0.9935>}
 pinky_01_r {parent=pinky_metacarpal_r, loc=<-7.1707 0.4832 0.1094>, rot=<-0.0654 -0.0035 0.2242 -0.9723>}
 pinky_02_r {parent=pinky_01_r, loc=<-2.8274 -0.0892 0.1043>, rot=<0.0037 0.0011 0.2854 -0.9584>}
 pinky_03_r {parent=pinky_02_r, loc=<-2.2045 -0.031 0.0078>, rot=<-0.0002 0.0098 0.1445 -0.9895>}
 ring_metacarpal_r {parent=hand_r, loc=<-2.2073 0.3627 -1.3638>, rot=<-0.0284 0.0722 -0.0458 -0.9959>}
 ring_01_r {parent=ring_metacarpal_r, loc=<-7.0297 0.6197 0.1446>, rot=<-0.0489 -0.0252 0.1892 -0.9804>}
 ring_02_r {parent=ring_01_r, loc=<-3.6171 -0.4648 0.1027>, rot=<0 0 0.2806 -0.9598>}
 ring_03_r {parent=ring_02_r, loc=<-2.4153 -0.1141 -0.0199>, rot=<0 0 0.1189 -0.9929>}
 deform_ring_01_rt {parent=ring_metacarpal_r, loc=<-7.6155 0.2399 1.3885>, rot=<-0.0907 -0.7101 0.6877 0.1209>}
 deform_ring_02_rt {parent=deform_ring_01_rt, loc=<3.4588 0 0>, rot=<0 -0.3356 0 -0.942>}
 deform_ring_03_rt {parent=deform_ring_02_rt, loc=<2.2137 0 0>, rot=<0 0 0 1>}
 thumb_01_r {parent=hand_r, loc=<-1.3119 0.3192 2.459>, rot=<0.6546 -0.4042 0.0664 -0.6354>}
 thumb_02_r {parent=thumb_01_r, loc=<-5.9335 0.0447 -0.0401>, rot=<-0.0037 0.0213 0.298 -0.9543>}
 thumb_03_r {parent=thumb_02_r, loc=<-2.2223 -0.3682 0.0907>, rot=<0.0118 -0.013 0.0066 -0.9998>}
 weapon_r {parent=hand_r, loc=<0.1975 3.4112 -0.3811>, rot=<0 0 0 -1>}
 deform_wrist_dn_base_r {parent=hand_r, loc=<0.0001 3.4173 0>, rot=<0 0 0 -1>}
 deform_wrist_dn_r {parent=deform_wrist_dn_base_r, loc=<0 0 0>, rot=<0 0 0 -1>}
 deform_wrist_up_base_r {parent=hand_r, loc=<0 -2.7803 0>, rot=<0 0 0 -1>}
 deform_wrist_up_r {parent=deform_wrist_up_base_r, loc=<0 0 0>, rot=<0 0 0 -1>}
 lowerarm_twist_01_r {parent=lowerarm_r, loc=<-17.855 0 0>, rot=<0 0 0 1>}
 lowerarm_twist_02_r {parent=lowerarm_r, loc=<-9.6199 0 0>, rot=<0 0 0 1>}
 deform_elbow_in_base_r {parent=lowerarm_r, loc=<-0.2114 -5.6449 -0.827>, rot=<0 0 0 1>}
 deform_elbow_in_r {parent=deform_elbow_in_base_r, loc=<0 0 0>, rot=<0 0 0 1>}
 deform_elbow_out_base_r {parent=lowerarm_r, loc=<-0.0001 7.5129 0.0002>, rot=<0 0 0 1>}
 deform_elbow_out_r {parent=deform_elbow_out_base_r, loc=<0 0 0>, rot=<0 0 0 1>}
 upperarm_twist_01_r {parent=upperarm_r, loc=<-7.9796 -0.0843 0.6195>, rot=<0 0 0 1>}
 deform_upperarm_bk_base_r {parent=upperarm_twist_01_r, loc=<6.8984 13.3827 1.3257>, rot=<0 0 0 1>}
 deform_upperarm_bk_r {parent=deform_upperarm_bk_base_r, loc=<0 0 0>, rot=<0 0 0 1>}
 deform_upperarm_fr_base_r {parent=upperarm_twist_01_r, loc=<8.3485 -10.7117 2.1457>, rot=<0 0 0 1>}
 deform_upperarm_fr_r {parent=deform_upperarm_fr_base_r, loc=<0 0 0>, rot=<0 0 0 1>}
 deform_upperarm_up_base_r {parent=upperarm_twist_01_r, loc=<11.2014 1.2235 -8.7787>, rot=<0 0 0 1>}
 deform_upperarm_up_r {parent=deform_upperarm_up_base_r, loc=<0 0 0>, rot=<0 0 0 1>}
 upperarm_twist_02_r {parent=upperarm_r, loc=<-19.506 -0.0843 0.6195>, rot=<0 0 0 1>}
 deform_bicep_base_r {parent=upperarm_twist_02_r, loc=<0.0002 -8.0494 -0.0002>, rot=<0 0 0 1>}
 deform_bicep_r {parent=deform_bicep_base_r, loc=<0 0 0>, rot=<0 0 0 1>}
 upperRetArm_r {parent=clavicle_r, loc=<-23.6836 0.248 2.4581>, rot=<0.0252 -0.2897 0.0018 -0.9568>}
 lowerRetArm_r {parent=upperRetArm_r, loc=<-23.1417 0 -0.0004>, rot=<0 0 -0.2611 -0.9653>}
 handRet_r {parent=lowerRetArm_r, loc=<-25.861 0.0001 0.0004>, rot=<0 0 0 1>}
 lowerRetArm_twist_01_r {parent=lowerRetArm_r, loc=<-6.4656 0.0002 0.0005>, rot=<0 0 0 1>}
 lowerRetArm_twist_02_r {parent=lowerRetArm_r, loc=<-12.9305 0 0.0002>, rot=<0 0 0 1>}
 lowerRetArm_twist_03_r {parent=lowerRetArm_r, loc=<-19.3955 0 -0.0001>, rot=<0 0 0 1>}
 lowerRetArm_twist_04_r {parent=lowerRetArm_r, loc=<-22.8804 0.0001 0.0005>, rot=<0 0 0 1>}
 upperRetArm_twist_01_r {parent=upperRetArm_r, loc=<-1.0001 0 0.0002>, rot=<0 0 0 -1>}
 upperRetArm_twist_02_r {parent=upperRetArm_r, loc=<-7.7139 0 -0.0001>, rot=<0 0 0 -1>}
 upperRetArm_twist_03_r {parent=upperRetArm_r, loc=<-15.4278 0 -0.0003>, rot=<0 0 0 -1>}
 neck_01 {parent=spine_05, loc=<10.815 1.9196 0>, rot=<0 0 -0.1864 -0.9825>}
 neck_02 {parent=neck_01, loc=<4.7347 0.3045 0>, rot=<0 0 0.0363 -0.9993>}
 head {parent=neck_02, loc=<5.7632 0.5552 0>, rot=<0 0 0.115 -0.9934>}
 attach_backpack {parent=spine_05, loc=<-0.1168 -13.475 0>, rot=<-0.1215 -0.6966 0.1215 -0.6966>}
 attach_cape {parent=spine_05, loc=<9.0127 -11.712 0>, rot=<0.1603 -0.6887 -0.1603 -0.6887>}
 Backpack_BR {parent=spine_05, type=Null, loc=<0 0 0>, rot=<0 0 0 1>}
 cape {parent=spine_05, type=Null, loc=<0 0 0>, rot=<0 0 0 1>}
 deform_pec_aim_base_l {parent=spine_04, loc=<5.0232 13.6719 -2.6508>, rot=<0.0672 -0.5201 0.3228 -0.7879>}
 deform_pec_aim_l {parent=deform_pec_aim_base_l, loc=<0 0 0>, rot=<0 0 0 -1>}
 deform_pec_base_l {parent=deform_pec_aim_l, loc=<16.3279 0 0>, rot=<0 0 0 -1>}
 deform_pec_l {parent=deform_pec_base_l, loc=<0 0 0>, rot=<0 0 0 -1>}
 deform_pec_aim_base_r {parent=spine_04, loc=<5.023 13.6718 2.6508>, rot=<-0.5201 -0.0672 0.7879 0.3228>}
 deform_pec_aim_r {parent=deform_pec_aim_base_r, loc=<0 0 0>, rot=<0 0 0 -1>}
 deform_pec_base_r {parent=deform_pec_aim_r, loc=<-16.3279 0 0>, rot=<0 0 0 -1>}
 deform_pec_r {parent=deform_pec_base_r, loc=<0 0 0.0002>, rot=<0 0 0 -1>}
 deform_lat_l_base {parent=spine_03, loc=<2.5566 -10.3716 -17.6918>, rot=<0 0 0 1>}
 deform_lat_l {parent=deform_lat_l_base, loc=<0 0 0>, rot=<0 0 0 1>}
 deform_lat_r_base {parent=spine_03, loc=<2.5568 -10.3716 17.6918>, rot=<0 0 1 0>}
 deform_lat_r {parent=deform_lat_r_base, loc=<0 0 0>, rot=<0 0 0 -1>}
 thigh_l {parent=pelvis, loc=<-1.8716 -0.0109 -11.9936>, rot=<0.0012 0.0572 0.0155 -0.9982>}
 calf_l {parent=thigh_l, loc=<-41.9178 -0.2246 0.0926>, rot=<0.0006 -0.0161 -0.0391 -0.9991>}
 calf_twist_01_l {parent=calf_l, loc=<-32.8646 -0.1379 0.08>, rot=<-0.0224 0.0096 -0.0042 -0.9997>}
 calf_twist_02_l {parent=calf_l, loc=<-15.865 -0.1398 0.1433>, rot=<0.0232 0.0114 -0.0042 -0.9997>}
 foot_l {parent=calf_l, loc=<-38.7057 1.2438 0.2452>, rot=<-0.0031 -0.0411 0.0414 -0.9983>}
 ball_l {parent=foot_l, loc=<-6.1614 13.1469 0.6807>, rot=<0.0001 -0.0001 -0.7063 -0.7079>}
 metaTarsoRet_l {parent=calf_l, loc=<-38.7057 1.2438 0.2452>, rot=<-0.0031 -0.0411 0.0414 -0.9983>}
 toeRet_l {parent=metaTarsoRet_l, loc=<-6.1614 11.5954 0.6807>, rot=<0 -0.0001 0 -1>}
 toeRet_l_end {parent=toeRet_l, loc=<0 10.9273 0>, rot=<0 0 0 1>}
 deform_knee_in_base_l {parent=calf_l, loc=<1.298 -8.0356 -0.0496>, rot=<0 0 0 1>}
 deform_knee_in_l {parent=deform_knee_in_base_l, loc=<0 0 0>, rot=<0 0 0 1>}
 deform_knee_out_base_l {parent=calf_l, loc=<0 6.8739 0>, rot=<0 0 0 1>}
 deform_knee_out_l {parent=deform_knee_out_base_l, loc=<0 0 0>, rot=<0 0 0 1>}
 thigh_twist_01_l {parent=thigh_l, loc=<-21.6303 0 0>, rot=<0.0013 0 0 -1>}
 deform_groin_base_l {parent=thigh_l, loc=<-2.1908 8.5661 -1.7442>, rot=<0 0 0 -1>}
 deform_groin_l {parent=deform_groin_base_l, loc=<0 0 0>, rot=<0 0 0 -1>}
 thigh_r {parent=pelvis, loc=<-1.8714 -0.0109 11.9936>, rot=<0.0572 -0.0012 0.9982 0.0155>}
 calf_r {parent=thigh_r, loc=<41.918 0.2246 -0.0926>, rot=<0.0006 -0.0161 -0.0391 -0.9991>}
 calf_twist_01_r {parent=calf_r, loc=<32.8677 0.117 -0.0912>, rot=<-0.2566 0.0146 -0.008 -0.9664>}
 calf_twist_02_r {parent=calf_r, loc=<15.8659 0.1418 -0.1572>, rot=<-0.3976 0.0095 -0.008 -0.9175>}
 foot_r {parent=calf_r, loc=<38.7058 -1.2438 -0.2453>, rot=<0.0975 -0.0474 0.0317 -0.9936>}
 ball_r {parent=foot_r, loc=<6.1614 -13.1469 -0.6807>, rot=<0.0001 -0.0001 -0.7063 -0.7079>}
 metaTarsoRet_r {parent=calf_r, loc=<38.7057 -1.2438 -0.2452>, rot=<0.0474 0.0975 0.9936 0.0317>}
 toeRet_r {parent=metaTarsoRet_r, loc=<-6.1614 11.5954 -0.6807>, rot=<0 0.0001 0 -1>}
 toeRet_r_end {parent=toeRet_r, loc=<0 10.9273 -0.0001>, rot=<0 0 0 -1>}
 deform_knee_in_base_r {parent=calf_r, loc=<-1.2981 8.0356 0.0496>, rot=<0 0 0 1>}
 deform_knee_in_r {parent=deform_knee_in_base_r, loc=<0 0 0>, rot=<0 0 0 1>}
 deform_knee_out_base_r {parent=calf_r, loc=<0 -6.8739 -0.0001>, rot=<0 0 0 1>}
 deform_knee_out_r {parent=deform_knee_out_base_r, loc=<0 0 0>, rot=<0 0 0 1>}
 thigh_twist_01_r {parent=thigh_r, loc=<21.6303 0 0>, rot=<0.0013 0 0 -1>}
 deform_groin_base_r {parent=thigh_r, loc=<2.191 -8.5661 1.7442>, rot=<0 0 0 -1>}
 deform_groin_r {parent=deform_groin_base_r, loc=<0 0 0>, rot=<0 0 0 -1>}
 thighRet_l {parent=pelvis, loc=<-8.1404 0.1831 -12.7143>, rot=<0.0444 0.0493 0.0322 -0.9973>}
 calfRet_l {parent=thighRet_l, loc=<-35.8247 0 0>, rot=<0 0 -0.0564 -0.9984>}
 footRet_l {parent=calfRet_l, loc=<-38.1283 0 0>, rot=<0 0 -0.0824 -0.9966>}
 calfRet_twist_01_l {parent=calfRet_l, loc=<-12.7094 0 0>, rot=<0 0 0 1>}
 calfRet_twist_02_l {parent=calfRet_l, loc=<-25.4189 0 0>, rot=<0 0 0 1>}
 thighRet_twist_01_l {parent=thighRet_l, loc=<-1 0 0>, rot=<0 0 0 1>}
 thighRet_twist_02_l {parent=thighRet_l, loc=<-11.9416 0 0>, rot=<0 0 0 1>}
 thighRet_twist_03_l {parent=thighRet_l, loc=<-23.8832 0 0>, rot=<0 0 0 1>}
 thighRet_r {parent=pelvis, loc=<-8.1404 0.1831 12.7143>, rot=<0.0493 -0.0444 0.9973 0.0322>}
 calfRet_r {parent=thighRet_r, loc=<35.8248 0 0>, rot=<0 0 -0.0564 -0.9984>}
 footRet_r {parent=calfRet_r, loc=<38.1283 0 0>, rot=<0 0 -0.0824 -0.9966>}
 calfRet_twist_01_r {parent=calfRet_r, loc=<12.7095 0 0>, rot=<0 0 0 -1>}
 calfRet_twist_02_r {parent=calfRet_r, loc=<25.4188 0 0>, rot=<0 0 0 -1>}
 thighRet_twist_01_r {parent=thighRet_r, loc=<1 0 0>, rot=<0 0 0 1>}
 thighRet_twist_02_r {parent=thighRet_r, loc=<11.9416 0 0>, rot=<0 0 0 1>}
 thighRet_twist_03_r {parent=thighRet_r, loc=<23.8832 0 0>, rot=<0 0 0 1>}
 deform_glute_base_l {parent=pelvis, loc=<-10.1877 -11.6217 -12.899>, rot=<0.0275 0.0501 -0.4472 -0.8926>}
 deform_glute_l {parent=deform_glute_base_l, loc=<0 0 0>, rot=<0 0 0 1>}
 deform_glute_base_r {parent=pelvis, loc=<-10.1877 -11.6218 12.899>, rot=<0.0501 -0.0275 0.8926 -0.4472>}
 deform_glute_r {parent=deform_glute_base_r, loc=<0 0 0>, rot=<0 0 0 1>}
 ik_foot_root {parent=root, loc=<0 0 0>, rot=<0 0 0 1>}
 ik_foot_l {parent=ik_foot_root, loc=<19.6189 -3.2051 10.3339>, rot=<0.0753 0.7018 0.0677 -0.7051>}
 ik_foot_r {parent=ik_foot_root, loc=<-19.6189 -3.2051 10.3338>, rot=<0.7051 0.0677 -0.7018 0.0753>}
 ik_hand_root {parent=root, loc=<0 0 0>, rot=<0 0 0 1>}
 ik_hand_gun {parent=ik_hand_root, loc=<-64.4621 4.8638 101.1961>, rot=<0.6084 0.1793 0.4613 -0.6205>}
 ik_hand_l {parent=ik_hand_gun, loc=<65.7709 101.9259 43.6659>, rot=<-0.5102 -0.7906 -0.3387 0>}
 ik_hand_r {parent=ik_hand_gun, loc=<0 0 0>, rot=<0 0 0 1>}
]
variables [
 ArmsIK_Softness {type=double, default=0}
 Entry_segmentAFactor {type=double, default=0}
 Entry_segmentBFactor {type=double, default=0}
 Entry_segmentsSum {type=double, default=0}
 isWearingGauntletsLeft {type=bool, default=False}
 isWearingGauntletsRight {type=bool, default=False}
]
