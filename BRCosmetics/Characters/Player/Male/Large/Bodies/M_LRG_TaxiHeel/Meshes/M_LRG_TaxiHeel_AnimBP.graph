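# Post-copy AnimBlueprint for the M_LRG_TaxiHeel large male character body, adapting the attached parent's pose to this skeleton.
# A Control Rig takes gauntlet flags from arm visibility events and IK softness from targeting/emote state, then feeds a cached pose.
# PickAxeEquip, updated on weapon change, selects between two arm-bone offset/squash chains, followed by fixed calf adjustments.
# The update event also computes scaled counter-rotations for head and neck.
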AnimBlueprint /BRCosmetics/Characters/Player/Male/Large/Bodies/M_LRG_TaxiHeel/Meshes/M_LRG_TaxiHeel_AnimBP
target_skeleton=&/M_LRG_TaxiHeel_Skeleton
graphs [
 AnimGraph {nodes [
   n1 = "Transform (Modify) Bone - Bone: lowerRetArm_twist_03_l"!(
    ComponentPose=Local To Component(LocalPose=Use cached pose 'Gauntlet_CACHE'())
    bAlphaBoolEnabled:bool=True
    Translation="0,2,4"
    Rotation="0,0,0"
    Scale="1,0.6,0.6"
   ) -> (Pose) {
    alphaInputType=Bool
    rotationMode=BMM_Additive
    rotationSpace=BCS_BoneSpace
    scaleMode=BMM_Additive
    scaleSpace=BCS_BoneSpace
    translationMode=BMM_Additive
    translationSpace=BCS_BoneSpace
    alphaBoolBlend {blendOption=Linear}
    boneToModify {boneName=lowerRetArm_twist_03_l}
    scale {y=0.6, z=0.6}
    translation {y=2, z=4}
   }
   n2 = "Transform (Modify) Bone - Bone: lowerRetArm_twist_04_l"!(ComponentPose=@n1, bAlphaBoolEnabled:bool=True, Translation="0,0,0", Rotation="0,0,0", Scale="1,0.5,0.5") -> (Pose) {
    alphaInputType=Bool
    rotationMode=BMM_Additive
    rotationSpace=BCS_BoneSpace
    scaleMode=BMM_Additive
    scaleSpace=BCS_BoneSpace
    translationMode=BMM_Additive
    translationSpace=BCS_BoneSpace
    alphaBoolBlend {blendOption=Linear}
    boneToModify {boneName=lowerRetArm_twist_04_l}
    scale {y=0.5, z=0.5}
   }
   n3 = "Transform (Modify) Bone - Bone: lowerRetArm_twist_03_r"!(ComponentPose=@n2, bAlphaBoolEnabled:bool=True, Translation="0,2,-4", Rotation="0,0,0", Scale="1,0.6,0.6") -> (Pose) {
    alphaInputType=Bool
    rotationMode=BMM_Additive
    rotationSpace=BCS_BoneSpace
    scaleMode=BMM_Additive
    scaleSpace=BCS_BoneSpace
    translationMode=BMM_Additive
    translationSpace=BCS_BoneSpace
    alphaBoolBlend {blendOption=Linear}
    boneToModify {boneName=lowerRetArm_twist_03_r}
    scale {y=0.6, z=0.6}
    translation {y=2, z=-4}
   }
   n4 = "Transform (Modify) Bone - Bone: lowerRetArm_twist_04_r"!(ComponentPose=@n3, bAlphaBoolEnabled:bool=True, Translation="0,0,0", Rotation="0,0,0", Scale="1,0.5,0.5") -> (Pose) {
    alphaInputType=Bool
    rotationMode=BMM_Additive
    rotationSpace=BCS_BoneSpace
    scaleMode=BMM_Additive
    scaleSpace=BCS_BoneSpace
    translationMode=BMM_Additive
    translationSpace=BCS_BoneSpace
    alphaBoolBlend {blendOption=Linear}
    boneToModify {boneName=lowerRetArm_twist_04_r}
    scale {y=0.5, z=0.5}
   }
   n5 = Component To Local(ComponentPose=@n4)
   n6 = "Transform (Modify) Bone - Bone: upperRetArm_twist_03_l"!(
    ComponentPose=Local To Component(LocalPose=Use cached pose 'Gauntlet_CACHE'())
    bAlphaBoolEnabled:bool=True
    Translation="0,-5,0"
    Rotation="0,0,0"
    Scale="1,0.6,0.6"
   ) -> (Pose) {
    alphaInputType=Bool
    rotationMode=BMM_Additive
    rotationSpace=BCS_BoneSpace
    scaleMode=BMM_Additive
    scaleSpace=BCS_BoneSpace
    translationMode=BMM_Additive
    translationSpace=BCS_BoneSpace
    alphaBoolBlend {blendOption=Linear}
    boneToModify {boneName=upperRetArm_twist_03_l}
    scale {y=0.6, z=0.6}
    translation {y=-5}
   }
   n7 = "Transform (Modify) Bone - Bone: lowerRetArm_l"!(ComponentPose=@n6, bAlphaBoolEnabled:bool=True, Translation="0,-2,0", Rotation="0,0,0", Scale="1,0.6,0.5") -> (Pose) {
    alphaInputType=Bool
    rotationMode=BMM_Additive
    rotationSpace=BCS_BoneSpace
    scaleMode=BMM_Additive
    scaleSpace=BCS_BoneSpace
    translationMode=BMM_Additive
    translationSpace=BCS_BoneSpace
    alphaBoolBlend {blendOption=Linear}
    boneToModify {boneName=lowerRetArm_l}
    scale {y=0.6, z=0.5}
    translation {y=-2}
   }
   n8 = "Transform (Modify) Bone - Bone: upperRetArm_twist_03_r"!(ComponentPose=@n7, bAlphaBoolEnabled:bool=True, Translation="0,5,0", Rotation="0,0,0", Scale="1,0.6,0.6") -> (Pose) {
    alphaInputType=Bool
    rotationMode=BMM_Additive
    rotationSpace=BCS_BoneSpace
    scaleMode=BMM_Additive
    scaleSpace=BCS_BoneSpace
    translationMode=BMM_Additive
    translationSpace=BCS_BoneSpace
    alphaBoolBlend {blendOption=Linear}
    boneToModify {boneName=upperRetArm_twist_03_r}
    scale {y=0.6, z=0.6}
    translation {y=5}
   }
   n9 = "Transform (Modify) Bone - Bone: lowerRetArm_r"!(ComponentPose=@n8, bAlphaBoolEnabled:bool=True, Translation="0,2,0", Rotation="0,0,0", Scale="1,0.6,0.5") -> (Pose) {
    alphaInputType=Bool
    rotationMode=BMM_Additive
    rotationSpace=BCS_BoneSpace
    scaleMode=BMM_Additive
    scaleSpace=BCS_BoneSpace
    translationMode=BMM_Additive
    translationSpace=BCS_BoneSpace
    alphaBoolBlend {blendOption=Linear}
    boneToModify {boneName=lowerRetArm_r}
    scale {y=0.6, z=0.5}
    translation {y=2}
   }
   n10 = Component To Local(ComponentPose=@n9)
   n11 = Get PickAxeEquip() -> (PickAxeEquip:bool)
   n12 = "Transform (Modify) Bone - Bone: calfRet_l"!(
    ComponentPose=Local To Component(LocalPose=Blend Poses by bool(BlendPose_0=@n5, BlendPose_1=@n10, bActiveValue=@n11))
    bAlphaBoolEnabled:bool=True
    Translation="0,0,-5"
    Rotation="0,0,0"
    Scale="1,0.3,0.1"
   ) -> (Pose) {
    alphaInputType=Bool
    rotationMode=BMM_Additive
    rotationSpace=BCS_BoneSpace
    scaleMode=BMM_Additive
    scaleSpace=BCS_BoneSpace
    translationMode=BMM_Additive
    translationSpace=BCS_BoneSpace
    alphaBoolBlend {blendOption=Linear}
    boneToModify {boneName=calfRet_l}
    scale {y=0.3, z=0.1}
    translation {z=-5}
   }
   n13 = "Transform (Modify) Bone - Bone: calfRet_r"!(ComponentPose=@n12, bAlphaBoolEnabled:bool=True, Translation="0,0,5", Rotation="0,0,0", Scale="1,0.3,0.1") -> (Pose) {
    alphaInputType=Bool
    rotationMode=BMM_Additive
    rotationSpace=BCS_BoneSpace
    scaleMode=BMM_Additive
    scaleSpace=BCS_BoneSpace
    translationMode=BMM_Additive
    translationSpace=BCS_BoneSpace
    alphaBoolBlend {blendOption=Linear}
    boneToModify {boneName=calfRet_r}
    scale {y=0.3, z=0.1}
    translation {z=5}
   }
   Output Pose(Result=Component To Local(ComponentPose=@n13)) {layerGroup=None}
   n14 = Control Rig!(
    Source=Copy Pose From Mesh() {bUseAttachedParent=true}
    Alpha
    ArmsIK_Softness:double=Get CtrlRig_IKArms_Softness() -> (CtrlRig_IKArms_Softness:double)
    isWearingGauntletsLeft:bool=Get isWearingGauntlets_left() -> (isWearingGauntlets_left:bool)
    isWearingGauntletsRight:bool=Get isWearingGauntlets_right() -> (isWearingGauntlets_right:bool)
   ) -> (Pose) {controlRigAssetReference {blueprintRigClass=&/M_LRG_TaxiHeel_CtrlRig.M_LRG_TaxiHeel_CtrlRig_C}}
   n15 = "Transform (Modify) Bone - Bone: neck_02"(
    ComponentPose="Transform (Modify) Bone - Bone: head"(ComponentPose=Local To Component(LocalPose=@n14), Alpha=1, Translation="0,0,0", Rotation="0,0,0", Scale="1,1,1") {rotationMode=BMM_Additive, rotationSpace=BCS_BoneSpace, boneToModify {boneName=head}}
    Alpha=1
    Translation="0,0,0"
    Rotation="0,0,0"
    Scale="1,1,1"
   ) {rotationMode=BMM_Additive, rotationSpace=BCS_BoneSpace, boneToModify {boneName=neck_02}}
   Save cached pose 'Gauntlet_CACHE'(Pose=Component To Local(ComponentPose="Transform (Modify) Bone - Bone: neck_01"(ComponentPose=@n15, Alpha=1, Translation="0,0,0", Rotation="0,0,0", Scale="1,1,1") {rotationMode=BMM_Additive, rotationSpace=BCS_BoneSpace, boneToModify {boneName=neck_01}})) {cacheName=Gauntlet_CACHE}
  ]}
 EventGraph {nodes [
   Event BlueprintUpdateAnimation()
   > n1 = Sequence()
   > n2 = Branch(Condition=OR Boolean(A=Get bIsTargeting(), B=Get bIsPlayingEmote()))
   > Set CtrlRig_IKArms_Softness(CtrlRig_IKArms_Softness:double=10) -> (Output_Get:double)
   Set CtrlRig_IKArms_Softness(execute=@n2.else, CtrlRig_IKArms_Softness=5)
   n3 = Get isCustomMontagePlaying() -> (isCustomMontagePlaying:bool)
   n4 = Get bIsPlayingEmote()
   Branch(
    execute=@n1.then_1
    Condition=AND Boolean(A=NOR Boolean!(A:bool=@n3, B:bool=@n4) -> (ReturnValue:bool), B=Get headLimit() -> (headLimit:bool))
   )
   > n5 = Sequence()
   n6 = GetOwnerSkeletalMeshComponent()
   n7 = BreakRotator(InRot=BreakTransform(InTransform=GetDeltaTransformFromRefPose(self=@n6, BoneName=head, BaseName=neck_02)).Rotation)
   n8 = Get X percentage() -> (X percentage:float)
   n9 = Get Y percentage() -> (Y percentage:float)
   n10 = Get Z percentage() -> (Z percentage:float)
   > Set head rot(head rot:Rotator=MakeRotator(
     Roll=float * float!(A=@n7, B=-1, C:double=@n8) -> (ReturnValue)
     Pitch=float * float(A=@n7.Pitch, B=-1, C=@n9)
     Yaw=float * float(A=@n7.Yaw, B=-1, C=@n10)
    )) -> (Output_Get:Rotator)
   n11 = GetOwnerSkeletalMeshComponent()
   n12 = BreakRotator(InRot=BreakTransform(InTransform=GetDeltaTransformFromRefPose(self=@n11, BoneName=neck_02, BaseName=neck_01)).Rotation)
   n13 = Get X Neck_02 Percentage() -> (X Neck_02 Percentage:double)
   n14 = Get Z Neck_02 Percentage() -> (Z Neck_02 Percentage:double)
   n15 = Get Y Neck_02 Percentage() -> (Y Neck_02 Percentage:double)
   > Set Neck_02 Rot(Neck_02 Rot:Rotator=MakeRotator(Roll=float * float(A=@n12, B=-1, C=@n13), Pitch=float * float(A=@n12.Pitch, B=-1, C=@n15), Yaw=float * float(A=@n12.Yaw, B=-1, C=@n14))) -> (Output_Get:Rotator)
   n16 = GetOwnerSkeletalMeshComponent()
   n17 = BreakRotator(InRot=BreakTransform(InTransform=GetDeltaTransformFromRefPose(self=@n16, BoneName=neck_01, BaseName=spine_05)).Rotation)
   n18 = Get Z Neck Percentage() -> (Z Neck Percentage:double)
   n19 = Get X Neck Percentage() -> (X Neck Percentage:double)
   n20 = Get Y Neck Percentage() -> (Y Neck Percentage:double)
   Set Neck rot(
    execute=@n5.then_1
    Neck rot:Rotator=MakeRotator(Roll=float * float(A=@n17, B=-1, C=@n19), Pitch=float * float(A=@n17.Pitch, B=-1, C=@n20), Yaw=float * float(A=@n17.Yaw, B=-1, C=@n18))
   ) -> (Output_Get:Rotator)
   Event BlueprintBeginPlay() {is=title K2Node_Event}
   > Set headLimit(headLimit:bool=true) -> (Output_Get:bool)
   n21 = Event OnCharacterCustomizationLoaded()
   > Set HeadMesh(HeadMesh:Skeletal Mesh Component Budgeted Object Reference=Get Array Item with EFort Custom Part Type(Array=@n21.CustomMeshParts, PartType=Head)) -> (Output_Get:Skeletal Mesh Component Budgeted Object Reference)
   > Set PickAxeEquip(PickAxeEquip:bool=true) -> (Output_Get:bool)
   Event OnLeftArmHidden() {is=title K2Node_Event}
   > Set isWearingGauntlets_left(isWearingGauntlets_left:bool=true) -> (Output_Get:bool)
   Event OnLeftArmVisible() {is=title K2Node_Event}
   > Set isWearingGauntlets_left(isWearingGauntlets_left=false)
   Event OnRightArmVisible() {is=title K2Node_Event}
   > Set isWearingGauntlets_right(isWearingGauntlets_right:bool=false) -> (Output_Get:bool)
   Event OnRightArmHidden() {is=title K2Node_Event}
   > Set isWearingGauntlets_right(isWearingGauntlets_right=true)
   Event OnWeaponChanged()
   > Set PickAxeEquip(PickAxeEquip=false)
   > Switch on EFortWeaponCoreAnimation!(execute, Selection:EFortWeaponCoreAnimation Enum=Get WeaponCoreAnim!() -> (WeaponCoreAnim)) -> (
    Melee:Exec
    Pistol:Exec
    Shotgun:Exec
    PaperBlueprint:Exec
    Rifle:Exec
    MeleeOneHand:Exec
    MachinePistol:Exec
    RocketLauncher:Exec
    GrenadeLauncher:Exec
    GoingCommando:Exec
    AssaultRifle:Exec
    TacticalShotgun:Exec
    SniperRifle:Exec
    TrapPlacement:Exec
    ShoulderLauncher:Exec
    AbilityDecoTool:Exec
    Crossbow:Exec
    C4:Exec
    RemoteControl:Exec
    DualWield:Exec
    AR_BullPup:Exec
    AR_ForwardGrip:Exec
    MedPackPaddles:Exec
    SMG_P90:Exec
    AR_DrumGun:Exec
    Consumable_Small:Exec
    Consumable_Large:Exec
    Balloon:Exec
    MountedTurret:Exec
    CreativeTool:Exec
    ExplosiveBow:Exec
    AshtonIndigo:Exec
    AshtonChicago:Exec
    MeleeDualWield:Exec
    Unarmed:Exec
   )
   > Set PickAxeEquip(PickAxeEquip=true)
   Event OnCharacterAndPickaxeCustomizationLoaded!() -> (
    OutputDelegate
    then
    CustomBaseMesh:Skeletal Mesh Component Object Reference
    CustomParts:Array of Custom Character Part Object References
    CustomMeshParts:Array of Skeletal Mesh Component Budgeted Object References
   ) {is=title K2Node_Event}
   Get Array Item with EFort Custom Part Type(Array:Array of Wildcards, PartType=Gameplay) -> (Output:Wildcard)
  ]}
]
variables [
 CtrlRig_IKArms_Softness {type=number, default=1}
 EndJump {type=boolean}
 head rot {type=Rotator}
 headLimit {type=boolean}
 HeadMesh {type=SkeletalMeshComponentBudgeted}
 isCustomMontagePlaying {type=boolean}
 isWearingGauntlets_left {type=boolean}
 isWearingGauntlets_right {type=boolean}
 Neck rot {type=Rotator}
 Neck_02 Rot {type=Rotator}
 PickAxeEquip {type=boolean}
 X Neck Percentage {type=number, default=0.2}
 X Neck_02 Percentage {type=number, default=0.2}
 X percentage {type=number, default=0.2}
 Y Neck Percentage {type=number, default=0.2}
 Y Neck_02 Percentage {type=number, default=0.2}
 Y percentage {type=number, default=0.2}
 Z Neck Percentage {type=number, default=0.2}
 Z Neck_02 Percentage {type=number, default=0.2}
 Z percentage {type=number, default=0.2}
]
